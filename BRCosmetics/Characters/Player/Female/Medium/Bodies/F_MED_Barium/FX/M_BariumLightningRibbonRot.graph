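# Translucent Niagara ribbon and sprite material for lightning effects on the Barium character body, with UVs rotated by CustomRotator.
# Panning caustic noise, masked across V and thresholded by the EnergyAmount dynamic parameter, shapes EnergyColor and particle color for emissive adjustment.
# Two panning dissolve textures, offset per ribbon, combine with sparkle-jittered edge falloff into an alpha-style mask.
# No material output links appear, and several branches, including the normal texture, feed nothing.
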Material /BRCosmetics/Characters/Player/Female/Medium/Bodies/F_MED_Barium/FX/M_BariumLightningRibbonRot
graph {
 properties {
  bUsedWithNiagaraRibbons=true
  bUsedWithNiagaraSprites=true
  blendMode=BLEND_Translucent
  pixelDepthOffsetMode=PDOM_Legacy
  usedShadingModels=Default Lit
 }
 nodes [
  n1 = Add(A=TextureCoordinate() {vTiling=0.5}, B=AppendVector(A=ScalarParameter() {parameterName=Param_1}, B=Constant() {r=0.3}))
  n2 = TextureCoordinate()
  n3 = ComponentMask(None=@n2) {g=true}
  n4 = Add(A=Multiply(A=OneMinus(None=@n3))) {constB=-0.1}
  n5 = Saturate(None=Multiply(A=@n3) {constB=10})
  n6 = ComponentMask(None=@n2) {r=true}
  n7 = Power(Base=@n6) {constExponent=1}
  n8 = Power(Base=OneMinus(None=@n6)) {constExponent=1}
  n9 = Multiply(A=Saturate(None=Multiply(A=@n4, B=@n5)), B=Multiply(A=Multiply(A=@n7, B=@n8)) {constB=0.3})
  n10 = Add(A=Multiply(A=TextureSample(UVs=@n1) {texture=/Game/UI/Foundation/Textures/Pattern/T_Sparkle}.G) {constB=0.02}, B=Add(A=@n9) {constB=0.02})
  n11 = Multiply(A=Power(Base=Saturate(None=@n10)) {constExponent=1}) {constB=20}
  n12 = MaterialFunctionCall(
   UVs=TextureCoordinate() {coordinateIndex=1}
   Rotation Center=Constant2Vector() {g=0.5, r=0.5}
   "Rotation Angle (0-1)"=ScalarParameter() {defaultValue=0.25, parameterName=RotationAngle}
  ) {materialFunction=&CustomRotator}
  n13 = ScalarParameter() {defaultValue=0.3, parameterName=Tile U Eataway}
  n14 = ScalarParameter() {defaultValue=1, parameterName=Tile V Eataway}
  n15 = ScalarParameter() {defaultValue=-2, parameterName=Pan U Eataway}
  n16 = ScalarParameter() {defaultValue=1, parameterName=Pan V Eataway}
  n17 = TextureSampleParameter2D(UVs=MaterialFunctionCall!(
    UV=@n12.Rotated Values
    Tile U=@n13
    Tile V=@n14
    Pan U=@n15
    Pan V=@n16
    Offset U=DynamicParameter() {paramNames [RandomPerRibbon, EnergyAmount, Param3, Param4]}.RandomPerRibbon
    Offset V
    Time
   ) {materialFunction=/Game/Effects/Fort_Effects/Materials/_Functions/MF_PanTileOffset}.UV's) {parameterName=PanTextureA, texture=/Game/CharClasses/Outlander/Gadgets/AntiMaterialCharge/Materials/T_AntiMaterialCharge_Dissolve_01}
  n18 = Multiply(A=OneMinus(None=@n17.G)) {constB=2}
  n19 = TextureSampleParameter2D(UVs=MaterialFunctionCall(
    UV=@n12.Rotated Values
    Tile U=Multiply(A=@n13) {constB=0.9}
    Tile V=Multiply(A=@n14) {constB=0.5}
    Pan U=Multiply(A=@n15) {constB=0.8}
    Pan V=Multiply(A=@n16) {constB=-0.2}
    Offset U=DynamicParameter() {paramNames [RandomPerRibbon, EnergyAmount, Param3, Param4]}.RandomPerRibbon
   ) {materialFunction=&MF_PanTileOffset}.UV's) {parameterName=PanTextureB, texture=&T_AntiMaterialCharge_Dissolve_01}
  n20 = Multiply(A=OneMinus(None=@n19.G))
  n21 = Multiply(A=@n18, B=@n20)
  Add(A=@n11, B=@n21)
  n22 = Power(Base=@n18) {constExponent=10}
  n23 = Power(Base=@n20) {constExponent=10}
  Add(A=@n22, B=@n23)
  n24 = ParticleColor()
  n25 = VectorParameter() {parameterName=EnergyColor, defaultValue=<rgba 1 0 0 1>}
  n26 = DynamicParameter() {paramNames [RandomPerRibbon, EnergyAmount, Param3, Param4]}
  n27 = Add(A=@n26.EnergyAmount) {constB=-0.25}
  n28 = Multiply(A=Multiply(A=@n25.RGB, B=ScalarParameter() {defaultValue=1, parameterName=EnergyBrightness}), B=Saturate(None=Multiply(A=@n27) {constB=1.5}))
  n29 = MaterialFunctionCall(
   UV=MaterialFunctionCall(
    UVs=TextureCoordinate() {coordinateIndex=1}
    Rotation Center=Constant2Vector() {g=0.5, r=0.5}
    "Rotation Angle (0-1)"=ScalarParameter() {defaultValue=0.25, parameterName=RotationAngle}
   ) {materialFunction=&CustomRotator}.Rotated Values
   Tile U=ScalarParameter() {defaultValue=0.1, parameterName="Tile U (S)"}
   Tile V=ScalarParameter() {defaultValue=1, parameterName="Tile V (S)"}
   Pan U=ScalarParameter() {defaultValue=-0.75, parameterName="Pan U (S)"}
   Offset U=DynamicParameter() {paramNames [RandomPerRibbon, EnergyAmount, Param3, Param4]}.RandomPerRibbon
  ) {materialFunction=&MF_PanTileOffset}
  n30 = TextureSampleParameter2D(UVs=@n29.UV's) {parameterName=NoiseTexture01, texture=&T_ShieldCaustics}
  n31 = ComponentMask(None=TextureCoordinate()) {g=true}
  n32 = Multiply(A=Power(Base=@n31) {constExponent=2.5}, B=Power(Base=OneMinus(None=@n31)) {constExponent=2.5})
  n33 = Multiply(A=Power(Base=Multiply(A=@n30.G, B=@n32)) {constExponent=1}) {constB=50}
  n34 = Saturate(None=SmoothStep(Min=LinearInterpolate(Alpha=@n26.EnergyAmount) {constB=0.0212}, Value=SmoothStep(Value=@n33) {constMax=0.03}) {constMin=0.2})
  n35 = Multiply(A=@n24.RGB, B=Multiply(A=@n28, B=@n34))
  Add(A=Multiply(A=@n24.RGB, B=@n28), B=@n35)
  LinearInterpolate(
   A=VectorParameter() {parameterName=OuterColorLow}.RGB
   B=@n25.RGB
   Alpha=DynamicParameter() {paramNames [RandomPerRibbon, EnergyAmount, Fill_Mult, Param4]}.EnergyAmount
  )
  LinearInterpolate(A=@n24.RGB, B=@n28, Alpha=@n34) {constB=0}
  MaterialFunctionCall() {materialFunction=&DebugTimeSine}
  MaterialFunctionCall(Emissive=@n35) {materialFunction=&MF_FX_EmissiveAdjustments}
  Multiply(A=@n22, B=@n23)
  Multiply(A=@n9) {constB=2}
  n36 = Multiply(A=Power(Base=@n9) {constExponent=2.2}) {constB=500}
  n37 = LinearInterpolate(A=@n11, B=@n21) {constAlpha=0.6}
  n38 = Saturate(None=Power(Base=Add(A=@n36, B=@n37)) {constExponent=30})
  Multiply(A=ParticleColor().A, B=SmoothStep(Value=Multiply(A=@n36, B=@n38)) {constMax=0.02})
  n39 = ParticleColor()
  Saturate(None=Add(A=@n17.G, B=LinearInterpolate(Alpha=@n39.A) {constA=-1}))
  TextureCoordinate() {coordinateIndex=1}
  TextureSampleParameter2D(UVs=@n29.UV's) {parameterName=NormalTexture, samplerType=SAMPLERTYPE_Normal, texture=/Game/Effects/Fort_Effects/Textures/Smoke/T_Smoke_Trail_Soft_N}
 ]
}
