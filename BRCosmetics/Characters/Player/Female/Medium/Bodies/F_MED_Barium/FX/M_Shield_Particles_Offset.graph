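# Translucent unlit Niagara sprite and mesh particle material for a glowing shield effect on the F_MED_Barium character.
# A panning, sine-pulsed grid texture in particle macro UVs combines with tinted edge glow from a panning gradient plus scene-normal Fresnel.
# Dynamic parameters drive Fresnel exponent, glow pan and grid UV offset.
# Opacity combines glow, particle alpha, sphere mask and nested depth fades confining it near intersecting geometry, with optional blinking.
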Material /BRCosmetics/Characters/Player/Female/Medium/Bodies/F_MED_Barium/FX/M_Shield_Particles_Offset
graph {
 properties {
  bUsedWithNiagaraMeshParticles=true
  bUsedWithNiagaraSprites=true
  blendMode=BLEND_Translucent
  pixelDepthOffsetMode=PDOM_Legacy
  shadingModel=MSM_Unlit
  usedShadingModels=Unlit
 }
 nodes [
  Constant2Vector()
  Constant()
  n1 = Multiply(
   A=ComponentMask(None=VectorParameter() {parameterName=MacroUV_XYScale, defaultValue=<rgba 1 1 0 0>}.RGB) {g=true, r=true}
   B=Multiply(A=ParticleMacroUV(), B=ScalarParameter() {defaultValue=5, parameterName=Scale MacroUVs}) {constB=5}
  )
  n2 = ComponentMask(None=Add(
    A=VectorParameter() {parameterName=MacroUV_Offset, defaultValue=<rgba 0 0 0 1>}.RGB
    B=DynamicParameter() {paramNames [FresnelExponent, Refraction, GlowPan, Param4]}.Param4
   )) {g=true, r=true}
  n3 = Time()
  n4 = ScalarParameter() {defaultValue=1, parameterName=MacroUV Panner Time Scale}
  n5 = VectorParameter() {parameterName=MacroUV Panner Speed}
  n6 = TextureSampleParameter2D(UVs=Panner(Coordinate=Add(A=@n1, B=@n2), Time=Multiply(A=@n3, B=@n4), Speed=ComponentMask(None=@n5.RGB) {g=true, r=true})) {
   parameterName=Texture
   samplerType=SAMPLERTYPE_LinearColor
   texture=/Game/Packages/Fortress_SharedMaterials/Textures/T_Slowfield_GridSquares_01
  }
  n7 = StaticComponentMaskParameter(None=@n6.RGBA) {defaultR=true, parameterName=MacroUV Texture Mask}
  n8 = StaticSwitchParameter(True=OneMinus(None=@n7), False=@n7) {defaultValue=true, parameterName=Use Inverse MacroUV Mask}
  n9 = StaticSwitchParameter(
   True=Multiply(A=Time(), B=ScalarParameter() {defaultValue=2, parameterName=MacroUV Sine Time Intensity}) {constB=2}
   False=ScalarParameter() {parameterName=No Pulse Sine Control}
  ) {defaultValue=true, parameterName=Use Pulsing MacroUV}
  n10 = Add(A=MaterialFunctionCall!(Sine Phase=Add(A=@n8, B=@n9), Value 1, Value 2) {materialFunction=/Engine/Functions/Engine_MaterialFunctions02/Utility/Sine_Remapped}.Remapped Sine) {constB=0.05}
  n11 = ParticleColor()
  n12 = Multiply(A=VectorParameter() {parameterName=EdgeGlow, defaultValue=<rgba 0.5208 2 1.8544 1>}.RGB, B=@n11.RGB)
  n13 = ComponentMask(None=VectorParameter() {parameterName=MaskUVscale, defaultValue=<rgba 0 0 0 1>}.RGB) {g=true, r=true}
  n14 = Divide(A=Subtract(A=ObjectPositionWS(), B=WorldPosition().XYZ), B=ScalarParameter() {defaultValue=100, parameterName=Mask_UV}) {constB=100}
  n15 = ComponentMask(None=@n14) {b=true}
  n16 = VectorParameter() {desc=Which Direction you would like the Mask to pan, group=Local Space Direction, parameterName=ObjectsVector, defaultValue=<rgba 0 1 0 1>}
  n17 = Normalize(VectorInput=Transform(None=@n16.RGB) {transformSourceType=TRANSFORMSOURCE_Local})
  n18 = Panner(
   Coordinate=LinearInterpolate(
    A=Multiply(A=@n13, B=@n15)
    B=DotProduct(A=@n17, B=@n14)
    Alpha=ScalarParameter() {group=Local Space Direction, parameterName=UseLocalSpaceDirection, sliderMax=1}
   )
   Time=DynamicParameter() {paramNames [FresnelExponent, Refraction, GlowPan, Param4]}.GlowPan
  ) {speedY=3}
  n19 = SceneTexture() {sceneTextureId=PPI_WorldNormal}
  n20 = Fresnel(
   ExponentIn=DynamicParameter() {paramNames [FresnelExponent, Refraction, GlowPan, Param4]}.FresnelExponent
   Normal=MaterialFunctionCall(Normal=ComponentMask(None=@n19.Color) {b=true, g=true, r=true}, Flatness=ScalarParameter() {defaultValue=0.2, parameterName=FlattenNormal}) {materialFunction=&FlattenNormal}.Result
  ) {baseReflectFraction=0, exponent=1}
  n21 = ShadingPathSwitch(Default=@n20, Deferred=@n20, Forward=VectorParameter() {parameterName=Forward, defaultValue=<rgba 0 0 1 1>}.RGB, Mobile=@n20)
  n22 = Add(
   A=Multiply(
    A=TextureSampleParameter2D(UVs=@n18) {parameterName=Mask_Texture, texture=/Game/Global/Textures/Default/Masks_and_Gradients/T_Vert_Gradient_Soft}.R
    B=ScalarParameter() {defaultValue=1.25, parameterName=GlowPanScale}
   ) {constB=1.25}
   B=ShadingPathSwitch(Default=@n21, Deferred=@n21, Mobile=Constant() {r=0.3}) {desc=Migrated from FeatureLevelSwitch}
  )
  MaterialFunctionCall(Emissive=Add(A=Multiply(A=@n10, B=@n11.RGB), B=Multiply(A=@n12, B=@n22))) {materialFunction=&MF_FX_EmissiveAdjustments}
  MaterialFunctionCall() {materialFunction=&DebugTimeSine}
  n23 = TextureCoordinate()
  n24 = Constant() {r=0.5}
  n25 = ScalarParameter() {defaultValue=0.5, parameterName=SphereMaskRadius}
  n26 = ScalarParameter() {parameterName=SphereMaskHardness}
  n27 = DepthFade(FadeDistance=ScalarParameter() {defaultValue=35, parameterName=VisibleDepthDistance}) {fadeDistanceDefault=80}
  n28 = Multiply(
   A=Multiply(A=Multiply(A=@n11.A, B=@n22), B=SphereMask(A=@n23, B=@n24, Radius=@n25, Hardness=@n26) {hardnessPercent=0})
   B=DepthFade(Opacity=OneMinus(None=@n27), FadeDistance=ScalarParameter() {defaultValue=15, parameterName=DepthFade}) {fadeDistanceDefault=15}
  )
  n29 = Time()
  n30 = ScalarParameter() {defaultValue=1, group=Blinking, parameterName=Time}
  n31 = Multiply(A=Sine(None=Multiply(A=@n29, B=@n30)), B=Constant() {r=0.5})
  n32 = Constant() {r=0.5}
  n33 = Clamp(None=Add(A=Add(A=@n31, B=@n32), B=ScalarParameter() {group=Blinking, parameterName=Force Max Sine}) {constA=1})
  Multiply(
   A=LinearInterpolate(A=@n28, B=Multiply(A=@n33, B=@n28), Alpha=ScalarParameter() {group=Blinking, parameterName=Use Blinking})
   B=ScalarParameter() {defaultValue=1, group=Blinking, parameterName=Opacity}
  )
 ]
}
