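# Unlit debug material for Niagara sprites in the F_MED_Barium character's FX folder, displaying a particle dynamic parameter value as on-screen digits.
# It copies the first dynamic parameter channel into all four components through AppendMany and feeds them to the engine's DebugFloat4Values function.
# The resulting color-coded readout passes through MF_FX_EmissiveAdjustments as its emissive input.
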Material /BRCosmetics/Characters/Player/Female/Medium/Bodies/F_MED_Barium/FX/M_TestValues
graph {
 properties {bUsedWithNiagaraSprites=true, pixelDepthOffsetMode=PDOM_Legacy, shadingModel=MSM_Unlit, usedShadingModels=Unlit}
 nodes [
  n1 = DynamicParameter()
  MaterialFunctionCall(Emissive=MaterialFunctionCall!(
    Vector4=MaterialFunctionCall(R=@n1.Param1, G=@n1.Param1, B=@n1.Param1, A=@n1.Param1) {materialFunction=&AppendMany}.RGBA
    MaximumNumberOfDigits
    UVs
    DebugTextLocation RG_UpperRight BA_LowerLeft
    Component Spacing
   ) {materialFunction=/Engine/Functions/Engine_MaterialFunctions02/Utility/DebugFloat4Values}.ColorCodedOutput) {materialFunction=&MF_FX_EmissiveAdjustments}
 ]
}
