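# Niagara system for the Barium female character cosmetic's idle effects, tinted by a user paint color and attached to the skeletal mesh.
# Head and shoulder sprite emitters use bone or socket sampling, with spawn gated by user toggles and limited to the front end.
# Helmet, jacket and leg ribbons read source-emitter positions through Particle Read to fade by distance.
# Smoothed owner speed drives ribbon width, with a cinematic-velocity override.
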NiagaraSystem /BRCosmetics/Characters/Player/Female/Medium/Bodies/F_MED_Barium/FX/NS_Barium_Char_Idle
dependencies {
 usedDataInterfaces [
  &NiagaraDataInterfaceCurve
  &NiagaraDataInterfaceVector2DCurve
  &NiagaraDataInterfaceArrayInt32
  &NiagaraDataInterfaceArrayFloat3
  &NiagaraDataInterfaceArrayFloat
  &NiagaraDataInterfaceParticleRead
 ]
 usedDynamicInputs [&MakeFloatFromVector]
 usedModules [
  &SystemState
  &self:SystemSpawnScript.SystemScriptSource.SystemScriptGraph.NiagaraNodeAssignment_1.SetVariables_#1
  &self:SystemSpawnScript.SystemScriptSource.SystemScriptGraph.NiagaraNodeAssignment_6.SetVariables_#2
  &self:SystemSpawnScript.SystemScriptSource.SystemScriptGraph.NiagaraNodeAssignment_0.SetVariables_#3
  &self:SystemSpawnScript.SystemScriptSource.SystemScriptGraph.NiagaraNodeAssignment_3.SetVariables_#4
  &self:SystemSpawnScript.SystemScriptSource.SystemScriptGraph.NiagaraNodeAssignment_2.SetVariables_#5
  &self:SystemSpawnScript.SystemScriptSource.SystemScriptGraph.NiagaraNodeAssignment_5.SetVariables_#6
  &self:SystemSpawnScript.SystemScriptSource.SystemScriptGraph.NiagaraNodeAssignment_4.SetVariables_#7
  &self:SystemSpawnScript.SystemScriptSource.SystemScriptGraph.NiagaraNodeAssignment_7.SetVariables_#8
  &self:SystemSpawnScript.SystemScriptSource.SystemScriptGraph.NiagaraNodeAssignment_8.SetVariables_#9
 ]
}
emitters {
 Decals_Head {
  emitterUpdateScript [
   &EmitterState()
   &SpawnRate(
    Spawn Probability=&Multiply_Float(A=$User.SpawnHelmetFX, B=&FN_FrontEndFloatModifier(Gameplay=0, Front End=1))
    SpawnRate=10
    Use Spawn Probability=true
   )
  ]
  particleSpawnScript [
   &InitializeParticle(
    Color=$User.BariumPaintColor
    Lifetime=3
    Sprite Rotation Mode=::NewEnumerator2
    Sprite Size Max=<50 50>
    Sprite Size Min=<50 20>
    Sprite Size Mode=::NewEnumerator2
    UsePositionOffset=false
    Write Material Random=true
   )
   &AddVelocity(Velocity=&RandomRangeVector(Minimum=<0 1 15>, Maximum=<0 -1 10>))
   &ApplyInitialForces()
   &SkeletalMeshLocation(
    Bone / Socket Index=&SelectIntFromArray(
     Array Sampling Mode=::NewEnumerator1
     Direct Array Index=?
     Int Selection Array=DataInterface() {propertyValues {IntData [0]}}
     Random Seed=0
    )
    MaskSpawnGroup=false
    Mesh Sampling Type=::NewEnumerator5
    Sampled Position Offset=&RandomRangeVector(Minimum=<-3 8 20>, Maximum=<3 12 20>)
    Sampled Position Offset Coordinate Space:ENiagaraSkelMeshTransforms=::NewEnumerator2
    Skeletal Mesh=$User.SkeletalMesh
    Socket Sampling Mode=::NewEnumerator5
   )
   &SubUVAnimation(End Frame=3, SubUV Animation Mode:ENiagara_SubUVLookupMode=::NewEnumerator2)
  ]
  particleUpdateScript [
   &ParticleState()
   &GravityForce() {enabled=false}
   &SolveForcesAndVelocity(Clamp Velocity=false, Limit Acceleration=false)
   &CameraOffset(Camera Offset Amount=&RandomRangeFloat(Minimum=-5, Maximum=-5)) {enabled=false}
   &ScaleColor(
    Scale Alpha=&FloatFromCurve(
     FloatCurve=DataInterface() {propertyValues {Curve {keys [
         {interpMode=RCIM_Cubic, tangentMode=RCTM_User, time=0.3, value=0}
         {interpMode=RCIM_Cubic, tangentMode=RCTM_User, time=0.6, value=0.75}
         {tangentMode=RCTM_User, time=1, value=0}
        ]}}}
     Scale Curve=0.1
    )
    Scale RGB=&VectorFromFloat(Value=1.5)
    ScaleA=true
    ScaleRGB=true
   )
   &ScaleSpriteSize(
    Non-Uniform Curve Sprite Scale=DataInterface() {propertyValues {
      XCurve {keys [
        {arriveTangent=-1.6995, interpMode=RCIM_Cubic, leaveTangent=-1.6995, tangentMode=RCTM_User, time=0, value=1}
        {interpMode=RCIM_Cubic, tangentMode=RCTM_User, time=1, value=0}
       ]}
      YCurve {keys [
        {arriveTangent=-0.0868, interpMode=RCIM_Cubic, leaveTangent=-0.0868, tangentMode=RCTM_User, time=0, value=0.5}
        {arriveTangent=1.2139, interpMode=RCIM_Cubic, leaveTangent=1.2139, tangentMode=RCTM_User, time=1, value=1}
       ]}
     }}
    Scale Sprite Size Mode=::NewEnumerator3
   )
   /SetVariables(
    Particles.Color2=&MultiplyLinearColor(
     Linear Color=&MakeLinearColorFromVectorAndFloat(
      "Vector (RGB)"=&Multiply_Vector(
       A=&MakeVectorFromLinearColorRGB(Color=&HueShiftLinearColor(
         Hue=0
         Saturation=&FloatFromCurve(FloatCurve=DataInterface() {propertyValues {Curve {keys [
              <at 0 0.1 RCIM_Cubic>
              {arriveTangent=0.9, interpMode=RCIM_Cubic, leaveTangent=0.9, time=0.2, value=1}
              {interpMode=RCIM_Cubic, leaveTangent=0.2637, time=1, value=1}
             ]}}})
         Value=&FloatFromCurve(
          FloatCurve=DataInterface() {propertyValues {Curve {keys [
              <at 0 0 RCIM_Cubic>
              {arriveTangent=8.4076, interpMode=RCIM_Cubic, leaveTangent=8.4076, time=0.1, value=2}
              {arriveTangent=-2.5641, interpMode=RCIM_Cubic, leaveTangent=-2.5641, time=0.5032, value=4.2311}
              <at 0.88 0 RCIM_Cubic>
             ]}}}
          Scale Curve=0.15
         )
         Linear Color=$Particles.Color
        ))
       B=&VectorFromFloat(Value=1.5)
      )
      "Float (Alpha)"=&FloatFromCurve(
       FloatCurve=DataInterface() {propertyValues {Curve {keys [
           {arriveTangent=-0.3136, interpMode=RCIM_Cubic, leaveTangent=-0.3136, tangentMode=RCTM_User, time=0, value=0}
           {arriveTangent=0.2209, interpMode=RCIM_Cubic, leaveTangent=0.2209, time=0.345, value=0.434}
           {arriveTangent=-0.6627, interpMode=RCIM_Cubic, leaveTangent=-0.6627, time=0.6682, value=0.1476}
           {interpMode=RCIM_Cubic, tangentMode=RCTM_User, time=1, value=0}
          ]}}}
       Scale Curve=0.02
      )
     )
     Scale Factor=&MakeVector4_FromFloat(Float=1)
    )
    Particles.Position2:NiagaraPosition=&Add_Vector(A=Linked() {linkedVariable {name=Particles.Position, type=NiagaraPosition}}, B=<0 0 -10>)
    Particles.SpriteSize2=&Vector2DFromCurve(
     Vector2Curve=DataInterface() {propertyValues {
       XCurve {keys [
         {interpMode=RCIM_Cubic, tangentMode=RCTM_Break, time=0, value=0}
         {arriveTangent=0.5, interpMode=RCIM_Cubic, leaveTangent=0.5, time=0.4, value=1}
         {arriveTangent=-1.4133, interpMode=RCIM_Cubic, leaveTangent=2.3465, tangentMode=RCTM_Break, time=1, value=0.5}
        ]}
       YCurve {keys [
         {arriveTangent=1.6347, interpMode=RCIM_Cubic, leaveTangent=1.6347, tangentMode=RCTM_User, time=0.1, value=0.175}
         {interpMode=RCIM_Cubic, tangentMode=RCTM_User, time=1, value=1.15}
        ]}
      }}
     ScaleCurve=&RandomRangeVector2D(Minimum=<5 30>, Maximum=<2 40>)
    )
   ) {isSetParameters=true, module=SetVariables_#10}
  ]
  rendererClasses [&NiagaraSpriteRendererProperties]
  renderers [
   {rendererClass=&NiagaraSpriteRendererProperties}
   {rendererClass=&NiagaraSpriteRendererProperties, rendererIndex=1}
   {rendererClass=&NiagaraSpriteRendererProperties, rendererIndex=2}
   {rendererClass=&NiagaraSpriteRendererProperties, rendererIndex=3}
  ]
 }
 Decals_Shoulders {
  emitterUpdateScript [
   &EmitterState(Life Cycle Mode=::NewEnumerator1, Loop Duration=3, UseLoopDelay=false)
   &SpawnRate(
    Spawn Probability=&Multiply_Float(A=$User.SpawnJacketFX, B=&FN_FrontEndFloatModifier(Gameplay=0, Front End=1))
    SpawnRate=40
    Use Spawn Probability=true
   )
  ]
  particleSpawnScript [
   &InitializeParticle(
    Color=$User.BariumPaintColor
    Lifetime=&RandomRangeFloat(Minimum=1, Maximum=2)
    Sprite Rotation Mode=::NewEnumerator2
    Sprite Size Max=<15 25>
    Sprite Size Min=<2 5>
    Sprite Size Mode=::NewEnumerator2
    UsePositionOffset=false
    Write Material Random=true
   )
   &AddVelocity(Rotation Coordinate Space=::World, Velocity=&RandomRangeVector(Minimum=<0 0 5>, Maximum=<0 0 5>))
   &ApplyInitialForces()
   &SubUVAnimation(End Frame=3, SubUV Animation Mode:ENiagara_SubUVLookupMode=::NewEnumerator2)
   &SkeletalMeshLocation(
    Bone / Socket Index=&SelectIntFromArray(
     Array Sampling Mode=::NewEnumerator1
     Direct Array Index=?
     Int Selection Array=DataInterface() {propertyValues {IntData [2,5,3,4,2,5,15,16]}}
     Random Seed=0
    )
    Bone Sampling Mode=::NewEnumerator4
    MaskSpawnGroup=false
    Sampled Position Offset=&RandomRangeVector(Minimum=<5 0 0>, Maximum=<-5 0 0>)
    Sampled Position Offset Coordinate Space:ENiagaraSkelMeshTransforms=::NewEnumerator2
    Skeletal Mesh=$User.SkeletalMesh
   ) {module=SkeletalMeshLocation001}
   &ShapeLocation(MaskSpawnGroup=false, Offset Mode=::NewEnumerator0, Override Seed=false, Sphere Radius=5, UseOwnerScale=true) {enabled=false}
  ]
  particleUpdateScript [
   &ParticleState()
   &GravityForce() {enabled=false}
   &CameraOffset(Camera Offset Amount=&RandomRangeFloat(Minimum=-4, Maximum=-2))
   &ScaleColor(
    Scale Alpha=&FloatFromCurve(
     FloatCurve=DataInterface() {propertyValues {Curve {keys [
         {interpMode=RCIM_Cubic, tangentMode=RCTM_User, time=0.3, value=0}
         {interpMode=RCIM_Cubic, tangentMode=RCTM_User, time=0.6113, value=0.3718}
         {arriveTangent=-0.9565, interpMode=RCIM_Cubic, leaveTangent=-0.9565, time=0.7371, value=0.0986}
         {tangentMode=RCTM_User, time=1, value=0}
        ]}}}
     Scale Curve=&RandomRangeFloat(Minimum=0.1, Maximum=0.4)
    )
    Scale RGB=&VectorFromFloat(Value=1.5)
    ScaleA=true
    ScaleRGB=true
   )
   &ScaleSpriteSize(
    Non-Uniform Curve Scale=<1 1.2>
    Non-Uniform Curve Sprite Scale=DataInterface() {propertyValues {
      XCurve {keys [{arriveTangent=-1.1981, interpMode=RCIM_Cubic, leaveTangent=-1.1981, tangentMode=RCTM_User, time=0, value=1}, <at 0.9957 0 RCIM_Cubic>]}
      YCurve {keys [
        {arriveTangent=-0.0868, interpMode=RCIM_Cubic, leaveTangent=-0.0868, tangentMode=RCTM_User, time=0, value=0.5}
        {arriveTangent=1.2139, interpMode=RCIM_Cubic, leaveTangent=1.2139, tangentMode=RCTM_User, time=1, value=1}
       ]}
     }}
    Scale Sprite Size Mode=::NewEnumerator3
   )
   &SkeletalMeshLocation(
    Bone / Socket Index=&SelectIntFromArray(
     Array Sampling Mode=::NewEnumerator1
     Direct Array Index=?
     Int Selection Array=DataInterface() {propertyValues {IntData [2,5,3,4,2,5,15,16]}}
     Random Seed=0
    )
    Bone Sampling Mode=::NewEnumerator4
    MaskSpawnGroup=false
    Sampled Position Offset=&RandomRangeVector(Minimum=<5 0 0>, Maximum=<-5 0 0>)
    Sampled Position Offset Coordinate Space:ENiagaraSkelMeshTransforms=::NewEnumerator2
    Sampled Velocity Coordinate Space:ENiagaraCoordinateSpace=::World
    Skeletal Mesh=$User.SkeletalMesh
   )
   /SetVariables(
    Particles.Position2:NiagaraPosition=&Add_Vector(A=Linked() {linkedVariable {name=Particles.Position, type=NiagaraPosition}}, B=<0 0 0>)
    Particles.SpriteSize2=&Vector2DFromCurve(
     Vector2Curve=DataInterface() {propertyValues {
       XCurve {keys [
         {interpMode=RCIM_Cubic, tangentMode=RCTM_Break, time=0, value=0}
         {arriveTangent=-1.1373, interpMode=RCIM_Cubic, leaveTangent=-1.1373, tangentMode=RCTM_User, time=0.6009, value=1.3638}
         {arriveTangent=-1.4133, interpMode=RCIM_Cubic, leaveTangent=2.3465, tangentMode=RCTM_Break, time=0.9971, value=-0.0824}
        ]}
       YCurve {keys [
         {arriveTangent=1.6347, interpMode=RCIM_Cubic, leaveTangent=1.6347, tangentMode=RCTM_User, time=0.1, value=0.175}
         {arriveTangent=1.3428, interpMode=RCIM_Cubic, leaveTangent=1.3428, tangentMode=RCTM_User, time=0.9961, value=1.6937}
        ]}
      }}
     ScaleCurve=&Multiply_Vector2D_ByFloat(
      Float=&SelectFloatFromArray(
       Array Sampling Mode=::NewEnumerator0
       Direct Array Index=?
       Float Selection Array=DataInterface() {propertyValues {FloatData [-1,1]}}
       Random Seed=0
      )
      Vector2D=&RandomRangeVector2D(Minimum=<5 20>, Maximum=<10 30>)
     )
    )
   ) {isSetParameters=true, module=SetVariables_#10}
   /SetVariables(Particles.Position=&Multiply_Vector(
     A=Linked() {linkedVariable {name=Particles.Position, type=NiagaraPosition}}
     B=&TransformVector(
      Vector=&MakeVector(
       X=1
       Y=1
       Z=&FloatFromCurve(FloatCurve=DataInterface() {propertyValues {Curve {keys [
            {interpMode=RCIM_Cubic, tangentMode=RCTM_User, time=0, value=1}
            {arriveTangent=0.0375, interpMode=RCIM_Cubic, leaveTangent=0.0375, tangentMode=RCTM_User, time=1, value=1.02}
           ]}}})
      )
      Source Space=::World
      Destination Space=::Local
     )
    )) {enabled=false, isSetParameters=true, module=SetVariables_#11}
   /SetVariables(Particles.Color2=&MultiplyLinearColor(
     Linear Color=&MakeLinearColorFromVectorAndFloat(
      "Vector (RGB)"=&Multiply_Vector(
       A=&MakeVectorFromLinearColorRGB(Color=&HueShiftLinearColor(
         Hue=0
         Saturation=&FloatFromCurve(FloatCurve=DataInterface() {propertyValues {Curve {keys [
              <at 0 0.1 RCIM_Cubic>
              {arriveTangent=0.9, interpMode=RCIM_Cubic, leaveTangent=0.9, time=0.2, value=1}
              {interpMode=RCIM_Cubic, leaveTangent=0.2637, time=1, value=1}
             ]}}})
         Value=&FloatFromCurve(
          FloatCurve=DataInterface() {propertyValues {Curve {keys [
              <at 0 0 RCIM_Cubic>
              {arriveTangent=8.4076, interpMode=RCIM_Cubic, leaveTangent=8.4076, time=0.1, value=2}
              {arriveTangent=-2.5641, interpMode=RCIM_Cubic, leaveTangent=-2.5641, time=0.5032, value=4.2311}
              <at 0.88 0 RCIM_Cubic>
             ]}}}
          Scale Curve=0.15
         )
         Linear Color=$Particles.Color
        ))
       B=&VectorFromFloat(Value=1.5)
      )
      "Float (Alpha)"=&Multiply_Float(
       A=&FloatFromCurve(
        FloatCurve=DataInterface() {propertyValues {Curve {keys [
            {arriveTangent=-0.3136, interpMode=RCIM_Cubic, leaveTangent=-0.3136, tangentMode=RCTM_User, time=0, value=1}
            {arriveTangent=1.0887, interpMode=RCIM_Cubic, leaveTangent=1.0887, time=0.3576, value=2}
            {arriveTangent=-1.5994, interpMode=RCIM_Cubic, leaveTangent=-1.5994, time=0.6807, value=1.7411}
            {interpMode=RCIM_Cubic, tangentMode=RCTM_User, time=1, value=1}
           ]}}}
        CurveIndex=$Emitter.LoopedAge
       )
       B=&FloatFromCurve(
        FloatCurve=DataInterface() {propertyValues {Curve {keys [
            {arriveTangent=-0.3136, interpMode=RCIM_Cubic, leaveTangent=-0.3136, tangentMode=RCTM_User, time=0, value=0}
            {arriveTangent=0.2209, interpMode=RCIM_Cubic, leaveTangent=0.2209, time=0.345, value=0.434}
            {arriveTangent=-0.6627, interpMode=RCIM_Cubic, leaveTangent=-0.6627, time=0.6682, value=0.1476}
            {interpMode=RCIM_Cubic, tangentMode=RCTM_User, time=1, value=0}
           ]}}}
        Scale Curve=0.02
       )
      )
     )
     Scale Factor=&MakeVector4_FromFloat(Float=1)
    )) {isSetParameters=true, module=SetVariables_#12}
   &SpriteRotationRate(Rotation Rate=&RandomRangeFloat(Minimum=-20, Maximum=20)) {enabled=false}
   &SolveForcesAndVelocity(Clamp Velocity=false, Limit Acceleration=false)
  ]
  rendererClasses [&NiagaraSpriteRendererProperties]
  renderers [
   {rendererClass=&NiagaraSpriteRendererProperties}
   {rendererClass=&NiagaraSpriteRendererProperties, rendererIndex=1}
   {rendererClass=&NiagaraSpriteRendererProperties, rendererIndex=2}
   {rendererClass=&NiagaraSpriteRendererProperties, rendererIndex=3}
  ]
 }
 Decals_Shoulders001 {
  emitterUpdateScript [
   &EmitterState(Life Cycle Mode=::NewEnumerator1, Loop Duration=3, UseLoopDelay=false)
   &SpawnRate(
    Spawn Probability=&Multiply_Float(A=$User.SpawnJacketFX, B=&FN_FrontEndFloatModifier(Gameplay=0, Front End=1))
    SpawnRate=40
    Use Spawn Probability=true
   )
  ]
  particleSpawnScript [
   &InitializeParticle(
    Color=$User.BariumPaintColor
    Lifetime=&RandomRangeFloat(Minimum=1, Maximum=0.5)
    Sprite Rotation Mode=::NewEnumerator2
    Sprite Size Max=<4 25>
    Sprite Size Min=<2 5>
    Sprite Size Mode=::NewEnumerator2
    UsePositionOffset=false
    Write Material Random=true
   )
   &AddVelocity(Rotation Coordinate Space=::World, Velocity=&RandomRangeVector(Minimum=<0 0 50>, Maximum=<0 0 50>))
   &ApplyInitialForces()
   &SubUVAnimation(End Frame=3, SubUV Animation Mode:ENiagara_SubUVLookupMode=::NewEnumerator2)
   &SkeletalMeshLocation(
    Bone / Socket Index=&SelectIntFromArray(
     Array Sampling Mode=::NewEnumerator1
     Direct Array Index=?
     Int Selection Array=DataInterface() {propertyValues {IntData [16,17,18,19]}}
     Random Seed=0
    )
    MaskSpawnGroup=false
    Mesh Sampling Type=::NewEnumerator5
    Sampled Position Offset=&RandomRangeVector(Minimum=<0 0 -5>, Maximum=<0 0 0>)
    Sampled Position Offset Coordinate Space:ENiagaraSkelMeshTransforms=::NewEnumerator1
    Skeletal Mesh=$User.SkeletalMesh
    Socket Sampling Mode=::NewEnumerator5
   )
   &ShapeLocation(MaskSpawnGroup=false, Offset Mode=::NewEnumerator0, Override Seed=false, Sphere Radius=5, UseOwnerScale=true) {enabled=false}
  ]
  particleUpdateScript [
   &ParticleState()
   &GravityForce() {enabled=false}
   &CameraOffset(Camera Offset Amount=&RandomRangeFloat(Minimum=-4, Maximum=-2)) {enabled=false}
   &ScaleColor(
    Scale Alpha=&FloatFromCurve(
     FloatCurve=DataInterface() {propertyValues {Curve {keys [
         {interpMode=RCIM_Cubic, tangentMode=RCTM_User, time=0.3, value=0}
         {interpMode=RCIM_Cubic, tangentMode=RCTM_User, time=0.6113, value=0.3718}
         {tangentMode=RCTM_User, time=1, value=0}
        ]}}}
     Scale Curve=&RandomRangeFloat(Minimum=0.2, Maximum=0.2)
    )
    Scale RGB=&VectorFromFloat(Value=1.5)
    ScaleA=true
    ScaleRGB=true
   )
   &ScaleSpriteSize(
    Non-Uniform Curve Scale=<1 1.2>
    Non-Uniform Curve Sprite Scale=DataInterface() {propertyValues {
      XCurve {keys [{arriveTangent=-1.1981, interpMode=RCIM_Cubic, leaveTangent=-1.1981, tangentMode=RCTM_User, time=0, value=1}, <at 0.9957 0 RCIM_Cubic>]}
      YCurve {keys [
        {arriveTangent=-0.0868, interpMode=RCIM_Cubic, leaveTangent=-0.0868, tangentMode=RCTM_User, time=0, value=0.5}
        {arriveTangent=1.2139, interpMode=RCIM_Cubic, leaveTangent=1.2139, tangentMode=RCTM_User, time=1, value=1}
       ]}
     }}
    Scale Sprite Size Mode=::NewEnumerator3
   ) {enabled=false}
   &SkeletalMeshLocation(
    Bone / Socket Index=&SelectIntFromArray(
     Array Sampling Mode=::NewEnumerator1
     Direct Array Index=?
     Int Selection Array=DataInterface() {propertyValues {IntData [12,13,16,17,18,19]}}
     Random Seed=0
    )
    MaskSpawnGroup=false
    Mesh Sampling Type=::NewEnumerator5
    Sampled Position Offset=&RandomRangeVector(Minimum=<0 0 -5>, Maximum=<0 0 0>)
    Sampled Position Offset Coordinate Space:ENiagaraSkelMeshTransforms=::NewEnumerator1
    Skeletal Mesh=$User.SkeletalMesh
    Socket Sampling Mode=::NewEnumerator5
   ) {module=SkeletalMeshLocation002}
   /SetVariables(
    Particles.Position2:NiagaraPosition=&Add_Vector(A=Linked() {linkedVariable {name=Particles.Position, type=NiagaraPosition}}, B=<0 0 0>)
    Particles.SpriteSize2=&Vector2DFromCurve(
     Vector2Curve=DataInterface() {propertyValues {
       XCurve {keys [
         {interpMode=RCIM_Cubic, tangentMode=RCTM_Break, time=0, value=0}
         {arriveTangent=-1.1373, interpMode=RCIM_Cubic, leaveTangent=-1.1373, tangentMode=RCTM_User, time=0.6009, value=1.3638}
         {arriveTangent=-1.4133, interpMode=RCIM_Cubic, leaveTangent=2.3465, tangentMode=RCTM_Break, time=0.9971, value=-0.0824}
        ]}
       YCurve {keys [
         {arriveTangent=1.6347, interpMode=RCIM_Cubic, leaveTangent=1.6347, tangentMode=RCTM_User, time=0.1, value=0.175}
         {arriveTangent=1.3428, interpMode=RCIM_Cubic, leaveTangent=1.3428, tangentMode=RCTM_User, time=0.9961, value=1.6937}
        ]}
      }}
     ScaleCurve=&Multiply_Vector2D_ByFloat(
      Float=&SelectFloatFromArray(
       Array Sampling Mode=::NewEnumerator0
       Direct Array Index=?
       Float Selection Array=DataInterface() {propertyValues {FloatData [-1,1]}}
       Random Seed=0
      )
      Vector2D=&RandomRangeVector2D(Minimum=<5 50>, Maximum=<10 20>)
     )
    )
   ) {isSetParameters=true, module=SetVariables_#10}
   /SetVariables(Particles.Position=&Multiply_Vector(
     A=Linked() {linkedVariable {name=Particles.Position, type=NiagaraPosition}}
     B=&TransformVector(
      Vector=&MakeVector(
       X=1
       Y=1
       Z=&FloatFromCurve(FloatCurve=DataInterface() {propertyValues {Curve {keys [
            {interpMode=RCIM_Cubic, tangentMode=RCTM_User, time=0, value=1}
            {arriveTangent=0.0375, interpMode=RCIM_Cubic, leaveTangent=0.0375, tangentMode=RCTM_User, time=1, value=1.02}
           ]}}})
      )
      Source Space=::World
      Destination Space=::Local
     )
    )) {enabled=false, isSetParameters=true, module=SetVariables_#11}
   /SetVariables(Particles.Color2=/Niagara/DynamicInputs/TypeConversions/MakeLinearColorFromFloat(Value:NiagaraFloat=&FloatFromCurve(
      FloatCurve=DataInterface() {propertyValues {Curve {keys [
          {arriveTangent=-0.3136, interpMode=RCIM_Cubic, leaveTangent=-0.3136, tangentMode=RCTM_User, time=0, value=0}
          {arriveTangent=0.2209, interpMode=RCIM_Cubic, leaveTangent=0.2209, time=0.345, value=0.434}
          {arriveTangent=-0.6627, interpMode=RCIM_Cubic, leaveTangent=-0.6627, time=0.6682, value=0.1476}
          {interpMode=RCIM_Cubic, tangentMode=RCTM_User, time=1, value=0}
         ]}}}
      Scale Curve=0.2
     ))) {isSetParameters=true, module=SetVariables_#12}
   &SpriteRotationRate(Rotation Rate=&RandomRangeFloat(Minimum=-20, Maximum=20)) {enabled=false}
   &SolveForcesAndVelocity(Clamp Velocity=false, Limit Acceleration=false)
   &SpriteFacingAndAlignment(Facing Coordinate Space=::Local, WriteAlignment=false, WriteFacing=true)
  ]
  rendererClasses [&NiagaraSpriteRendererProperties]
  renderers [
   {rendererClass=&NiagaraSpriteRendererProperties}
   {rendererClass=&NiagaraSpriteRendererProperties, rendererIndex=1}
   {rendererClass=&NiagaraSpriteRendererProperties, rendererIndex=2}
   {rendererClass=&NiagaraSpriteRendererProperties, rendererIndex=3}
  ]
 }
 HelmetRibbons {
  emitterUpdateScript [
   &EmitterState()
   /SetVariables(Emitter.RibbonLengthModifier:NiagaraFloat=&Multiply_Float(A=1, B=$User.RibbonLengthFactor)) {isSetParameters=true, module=SetVariables_#13}
   &InterpolateOverTime(
    Rate Of Change:NiagaraFloat=&MakeCustomFloatFromBool(Bool=&SetBoolByFloatComparison(A=$System.EngineOwnerSpeed, B=0.5), True Float=1, False Float=15)
    Target Float Value:NiagaraFloat=$System.EngineOwnerSpeed
   )
  ]
  particleSpawnScript [
   /SetVariables(Particles.MaxRibbonWidth=24) {isSetParameters=true, module=SetVariables_#14}
   &InitializeRibbon(
    Color=&MultiplyLinearColor(
     Linear Color=&HueShiftLinearColor(Hue=0, Saturation=0.9, Value=2, Linear Color=$User.BariumPaintColor)
     Scale Factor=<1 1 1 0.05>
    )
    Lifetime=0.2
    Ribbon Width=&FloatFromCurve(
     FloatCurve=DataInterface() {propertyValues {Curve {keys [
         {arriveTangent=4.1785, interpMode=RCIM_Cubic, leaveTangent=4.1785, tangentMode=RCTM_User, time=0, value=0}
         {arriveTangent=0.9995, interpMode=RCIM_Cubic, leaveTangent=0.9995, tangentMode=RCTM_User, time=0.2223, value=0.6825}
         {arriveTangent=-0.0405, interpMode=RCIM_Cubic, leaveTangent=-0.0405, tangentMode=RCTM_User, time=1, value=1}
        ]}}}
     CurveIndex=&ClampFloat(Float=&Divide_Float(A=$Emitter.InterpolateOverTime.InterpolateOverTime_Float.Moving Average, B=100))
     Scale Curve=$Particles.MaxRibbonWidth
    )
    UsePositionOffset=false
    Write Material Random=true
   )
   /SetVariables(Particles.EventPosition:NiagaraPosition=$Particles.Position) {isSetParameters=true, module=SetVariables_#15}
   /SetVariables(Particles.SpriteSize=<5 5>) {isSetParameters=true, module=SetVariables_#16}
   /SetVariables(Particles.MaxRibbonLength:NiagaraFloat=&Multiply_Float(A=20, B=$Emitter.RibbonLengthModifier)) {isSetParameters=true, module=SetVariables_#17}
   /SetVariables() {isSetParameters=true, module=SetVariables_#18}
   /SetVariables() {isSetParameters=true, module=SetVariables_#19}
   /SetVariables() {isSetParameters=true, module=SetVariables_#20}
  ]
  particleUpdateScript [
   &ParticleState()
   &Drag(Drag=16, Use Linear Drag=true, Use Rotational Drag=false)
   &PointAttractionForce(
    Attraction Radius=1000
    Attraction Strength=&Multiply_Float(
     A=80
     B=&ClampFloat(Float=&Divide_Float(A=$Emitter.InterpolateOverTime.InterpolateOverTime_Float.Moving Average, B=100))
    )
    Attractor Position Offset:Vector3f=<0 -25 100>
    Attractor Position Offset Coordinate Space:ENiagaraCoordinateSpace=::Local
    Kill Within Radius=false
    Module.Use Falloff=false
    UseAttractorOffset=true
   ) {enabled=false}
   &SolveForcesAndVelocity(Clamp Velocity=false, Limit Acceleration=false)
   /SetVariables(Particles.NormalizedDistanceFromParent:NiagaraFloat=&ClampFloat(Float=&Divide_Float(
      A=&VectorLength(
       Vector=&Subtract_Vector(
        A=Linked() {linkedVariable {name=Particles.Position, type=NiagaraPosition}}
        B=&self:GetParentParticlePosition(PAR:NiagaraDataInterfaceParticleRead=DataInterface() {propertyValues {EmitterBinding {bindingMode=Other, emitterName=RibbonSourcesHelmet}}})
       )
       Fallback Vector=<0 0 0.01>
       Fallback Length=0
      )
      B=$Particles.MaxRibbonLength
     ))) {isSetParameters=true, module=SetVariables_#21}
   &ScaleColor(
    Scale Alpha=&MakeCustomFloatFromBool(
     Bool=$Particles.EventReceieved
     True Float=&FloatFromCurve(FloatCurve=DataInterface() {propertyValues {Curve {keys [<at 0 1>,<at 0.8 0>]}}}, CurveIndex=$Particles.NormalizedDistanceFromParent)
     False Float=1
    )
    Scale RGB=&VectorFromFloat(Value=&FloatFromCurve(
      FloatCurve=DataInterface() {propertyValues {Curve {keys [
          {arriveTangent=-2.725, interpMode=RCIM_Cubic, leaveTangent=-2.725, tangentMode=RCTM_User, time=0, value=5}
          {arriveTangent=-8.8847, leaveTangent=-8.8847, time=0.2, value=0.2}
          <at 1 0>
         ]}}}
      CurveIndex=$Particles.NormalizedDistanceFromParent
     ))
    ScaleA=true
    ScaleRGB=true
   )
   &ScaleColor(
    Scale Alpha=&Multiply_Float(
     A=&Lerp_Float(A=0, B=0.2, Alpha=&Divide_Float(A=$Emitter.InterpolateOverTime.InterpolateOverTime_Float.Moving Average, B=100))
     B=&FloatFromCurve(FloatCurve=DataInterface() {propertyValues {Curve {keys [<at 0 1>,<at 1 0>]}}})
    )
    ScaleA=true
    ScaleRGB=true
   ) {module=ScaleColor002}
   /SetVariables(Particles.Color2=&MultiplyLinearColor(
     Linear Color=$User.BariumPaintColor
     Scale Factor=&MakeFloat4FromVectorAndFloat(
      "Vector (RGB)"=&VectorFromFloat(Value=&FloatFromCurve(
        FloatCurve=DataInterface() {propertyValues {Curve {keys [<at 0 1>,<at 0.15 0.1>]}}}
        CurveIndex=$Particles.NormalizedDistanceFromParent
        Scale Curve=0.2
       ))
      "Float (Alpha)"=&Multiply_Float(
       A=&FloatFromCurve(FloatCurve=DataInterface() {propertyValues {Curve {keys [
            {arriveTangent=-3.2069, interpMode=RCIM_Cubic, leaveTangent=-3.2069, tangentMode=RCTM_User, time=0, value=1}
            {interpMode=RCIM_Cubic, tangentMode=RCTM_User, time=1, value=0}
           ]}}})
       B=&FloatFromCurve(
        FloatCurve=DataInterface() {propertyValues {Curve {keys [
            {arriveTangent=-3.6157, interpMode=RCIM_Cubic, leaveTangent=-3.6157, tangentMode=RCTM_User, time=0, value=1}
            <at 0.1 0.1>
            {interpMode=RCIM_Cubic, tangentMode=RCTM_User, time=1, value=0}
           ]}}}
        CurveIndex=$System.NormalizedPlayerSpeed
       )
      )
     )
    )) {isSetParameters=true}
   &KillParticles(Kill Particles=&Boolean_AndOperation(
     Bool1=&SetBoolByFloatComparison(
      A=&MakeFloatFromLinearColor(LinearColor=$Particles.Color, Channel:ENiagaraLinearColor_Channels=::NewEnumerator3)
      B=0.1
      Comparison Type=::NewEnumerator6
     )
     Bool2=$Particles.EventReceieved
    )) {enabled=false, module=KillParticles002}
   /SetVariables(Particles.RibbonU0Override=&MakeCustomFloatFromBool(Bool=$Particles.EventReceieved, True Float=$Particles.NormalizedDistanceFromParent, False Float=0)) {isSetParameters=true, module=SetVariables_#22}
   &ScaleRibbonWidth(Ribbon Width Scale=&FloatFromCurve(
     FloatCurve=DataInterface() {propertyValues {Curve {keys [
         {arriveTangent=1.3072, interpMode=RCIM_Cubic, leaveTangent=1.3072, tangentMode=RCTM_User, time=0, value=0}
         {arriveTangent=0.1238, interpMode=RCIM_Cubic, leaveTangent=0.1238, tangentMode=RCTM_User, time=1, value=1}
        ]}}}
     CurveIndex=&MakeCustomFloatFromBool(
      Bool=$User.UseCinematicVelocity
      True Float=1
      False Float=&ClampFloat(Float=&Divide_Float(A=$Emitter.InterpolateOverTime.InterpolateOverTime_Float.Moving Average, B=100))
     )
     Scale Curve=&MakeCustomFloatFromBool(
      Bool=$Particles.EventReceieved
      True Float=&FloatFromCurve(
       FloatCurve=DataInterface() {propertyValues {Curve {keys [{tangentMode=RCTM_User, time=0, value=1}, {tangentMode=RCTM_User, time=1, value=0}]}}}
       CurveIndex=$Particles.NormalizedDistanceFromParent
      )
      False Float=1
     )
    ))
  ]
  rendererClasses [&NiagaraRibbonRendererProperties, &NiagaraSpriteRendererProperties]
  renderers [
   {rendererClass=&NiagaraRibbonRendererProperties}
   {rendererClass=&NiagaraSpriteRendererProperties, rendererIndex=1}
   {rendererClass=&NiagaraRibbonRendererProperties, rendererIndex=2}
   {rendererClass=&NiagaraRibbonRendererProperties, rendererIndex=3}
   {rendererClass=&NiagaraRibbonRendererProperties, rendererIndex=4}
  ]
 }
 JacketRibbons002 {
  emitterUpdateScript [
   &EmitterState()
   /SetVariables(Emitter.RibbonLengthModifier:NiagaraFloat=&Multiply_Float(
     A=&FloatFromCurve(
      FloatCurve=DataInterface() {propertyValues {Curve {keys [
          {interpMode=RCIM_Cubic, tangentMode=RCTM_User, time=0, value=0}
          {arriveTangent=2.3919, interpMode=RCIM_Cubic, leaveTangent=2.3919, tangentMode=RCTM_User, time=1, value=1}
         ]}}}
      CurveIndex=$System.NormalizedPlayerSpeed
      Scale Curve=2
     )
     B=$User.RibbonLengthFactor
    )) {isSetParameters=true, module=SetVariables_#13}
   &InterpolateOverTime(
    Rate Of Change:NiagaraFloat=&MakeCustomFloatFromBool(Bool=&SetBoolByFloatComparison(A=$System.EngineOwnerSpeed, B=1), True Float=0.5, False Float=150)
    Target Float Value:NiagaraFloat=$System.EngineOwnerSpeed
   )
  ]
  particleSpawnScript [
   /SetVariables(Particles.MaxRibbonWidth=20) {isSetParameters=true, module=SetVariables_#14}
   &InitializeRibbon(
    Color=&MultiplyLinearColor(
     Linear Color=&HueShiftLinearColor(Hue=0, Saturation=1, Value=1, Linear Color=$User.BariumPaintColor)
     Scale Factor=<1 1 1 1>
    )
    Lifetime=0.15
    Ribbon Width=&FloatFromCurve(
     FloatCurve=DataInterface() {propertyValues {Curve {keys [
         {arriveTangent=4.1785, interpMode=RCIM_Cubic, leaveTangent=4.1785, tangentMode=RCTM_User, time=0, value=0}
         {arriveTangent=0.9995, interpMode=RCIM_Cubic, leaveTangent=0.9995, tangentMode=RCTM_User, time=0.2223, value=0.6825}
         {arriveTangent=-0.0405, interpMode=RCIM_Cubic, leaveTangent=-0.0405, tangentMode=RCTM_User, time=1, value=1}
        ]}}}
     CurveIndex=&ClampFloat(Float=&Divide_Float(A=$Emitter.InterpolateOverTime.InterpolateOverTime_Float.Moving Average, B=100))
     Scale Curve=$Particles.MaxRibbonWidth
    )
    UsePositionOffset=false
    Write Material Random=true
   )
   /SetVariables(Particles.EventPosition:NiagaraPosition=$Particles.Position) {isSetParameters=true, module=SetVariables_#15}
   /SetVariables(Particles.SpriteSize=<5 5>) {isSetParameters=true, module=SetVariables_#16}
   /SetVariables(Particles.MaxRibbonLength:NiagaraFloat=&Multiply_Float(A=10, B=$Emitter.RibbonLengthModifier)) {isSetParameters=true, module=SetVariables_#17}
   /SetVariables() {isSetParameters=true, module=SetVariables_#18}
   /SetVariables() {isSetParameters=true, module=SetVariables_#19}
   /SetVariables() {isSetParameters=true, module=SetVariables_#20}
  ]
  particleUpdateScript [
   &ParticleState()
   &Drag(Drag=50, Use Linear Drag=true, Use Rotational Drag=false)
   &PointAttractionForce(
    Attraction Radius=800
    Attraction Strength=&Multiply_Float(
     A=&Lerp_Float(A=0, B=-400, Alpha=$System.NormalizedPlayerSpeed)
     B=&ClampFloat(Float=&Divide_Float(A=$Emitter.InterpolateOverTime.InterpolateOverTime_Float.Moving Average, B=100))
    )
    Attractor Position Offset:Vector3f=<0 0 100>
    Attractor Position Offset Coordinate Space:ENiagaraCoordinateSpace=::Local
    Kill Within Radius=false
    Module.Use Falloff=false
    UseAttractorOffset=true
   ) {enabled=false}
   &SolveForcesAndVelocity(Clamp Velocity=false, Limit Acceleration=false)
   /SetVariables(Particles.NormalizedDistanceFromParent:NiagaraFloat=&ClampFloat(Float=&Divide_Float(
      A=&VectorLength(
       Vector=&Subtract_Vector(
        A=Linked() {linkedVariable {name=Particles.Position, type=NiagaraPosition}}
        B=&self:GetParentParticlePosition_12(PAR:NiagaraDataInterfaceParticleRead=DataInterface() {propertyValues {EmitterBinding {bindingMode=Other, emitterName=RibbonSourcesJacket}}})
       )
       Fallback Vector=<0 0 0.01>
       Fallback Length=0
      )
      B=$Particles.MaxRibbonLength
     ))) {isSetParameters=true, module=SetVariables_#21}
   &ScaleColor(
    Scale Alpha=&MakeCustomFloatFromBool(
     Bool=$Particles.EventReceieved
     True Float=&FloatFromCurve(FloatCurve=DataInterface() {propertyValues {Curve {keys [<at 0 1>,<at 0.8 0>]}}}, CurveIndex=$Particles.NormalizedDistanceFromParent)
     False Float=1
    )
    Scale RGB=&VectorFromFloat(Value=&FloatFromCurve(
      FloatCurve=DataInterface() {propertyValues {Curve {keys [{arriveTangent=-2.725, time=0, value=1}, {arriveTangent=-8.8847, leaveTangent=-8.8847, time=0.5, value=0.2}, <at 1 0>]}}}
      CurveIndex=$Particles.NormalizedDistanceFromParent
     ))
    ScaleA=true
    ScaleRGB=true
   )
   &ScaleColor(
    Scale Alpha=&FloatFromCurve(
     FloatCurve=DataInterface() {propertyValues {Curve {keys [
         {interpMode=RCIM_Cubic, tangentMode=RCTM_User, time=0, value=1}
         <at 0.2 0.05>
         {interpMode=RCIM_Cubic, tangentMode=RCTM_User, time=1, value=0}
        ]}}}
     Scale Curve=0.1
    )
    Scale RGB=&VectorFromFloat(Value=&FloatFromCurve(FloatCurve=DataInterface() {propertyValues {Curve {keys [<at 0 1>,<at 0.15 0.1>]}}}, CurveIndex=$Particles.NormalizedDistanceFromParent))
    ScaleA=true
    ScaleRGB=true
   ) {module=ScaleColor002}
   /SetVariables(Particles.Color2=&MultiplyLinearColor(
     Linear Color=$User.BariumPaintColor
     Scale Factor=&MakeFloat4FromVectorAndFloat(
      "Vector (RGB)"=&VectorFromFloat(Value=&FloatFromCurve(
        FloatCurve=DataInterface() {propertyValues {Curve {keys [
            {arriveTangent=-3.8955, interpMode=RCIM_Cubic, leaveTangent=-3.8955, tangentMode=RCTM_User, time=0, value=1}
            {interpMode=RCIM_Cubic, tangentMode=RCTM_User, time=1, value=0}
           ]}}}
        CurveIndex=$Particles.NormalizedDistanceFromParent
        Scale Curve=0.2
       ))
      "Float (Alpha)"=&Multiply_Float(
       A=&FloatFromCurve(FloatCurve=DataInterface() {propertyValues {Curve {keys [
            {arriveTangent=-3.2069, interpMode=RCIM_Cubic, leaveTangent=-3.2069, tangentMode=RCTM_User, time=0, value=1}
            {interpMode=RCIM_Cubic, tangentMode=RCTM_User, time=1, value=0}
           ]}}})
       B=0.5
      )
     )
    )) {isSetParameters=true}
   &KillParticles(Kill Particles=&Boolean_AndOperation(
     Bool1=&SetBoolByFloatComparison(
      A=&MakeFloatFromLinearColor(LinearColor=$Particles.Color, Channel:ENiagaraLinearColor_Channels=::NewEnumerator3)
      B=0.02
      Comparison Type=::NewEnumerator6
     )
     Bool2=$Particles.EventReceieved
    )) {enabled=false, module=KillParticles002}
   /SetVariables(Particles.RibbonU0Override=&MakeCustomFloatFromBool(Bool=$Particles.EventReceieved, True Float=$Particles.NormalizedDistanceFromParent, False Float=0)) {isSetParameters=true, module=SetVariables_#22}
   &ScaleRibbonWidth(Ribbon Width Scale=&FloatFromCurve(
     FloatCurve=DataInterface() {propertyValues {Curve {keys [
         {arriveTangent=1.3072, interpMode=RCIM_Cubic, leaveTangent=1.3072, tangentMode=RCTM_User, time=0, value=0}
         {arriveTangent=0.1238, interpMode=RCIM_Cubic, leaveTangent=0.1238, tangentMode=RCTM_User, time=1, value=1}
        ]}}}
     CurveIndex=&MakeCustomFloatFromBool(
      Bool=$User.UseCinematicVelocity
      True Float=1
      False Float=&ClampFloat(Float=&Divide_Float(A=$Emitter.InterpolateOverTime.InterpolateOverTime_Float.Moving Average, B=100))
     )
     Scale Curve=&MakeCustomFloatFromBool(
      Bool=$Particles.EventReceieved
      True Float=&FloatFromCurve(
       FloatCurve=DataInterface() {propertyValues {Curve {keys [{tangentMode=RCTM_User, time=0, value=1}, {tangentMode=RCTM_User, time=1, value=1}]}}}
       CurveIndex=$Particles.NormalizedDistanceFromParent
      )
      False Float=1
     )
    ))
  ]
  rendererClasses [&NiagaraRibbonRendererProperties, &NiagaraSpriteRendererProperties]
  renderers [
   {rendererClass=&NiagaraRibbonRendererProperties}
   {rendererClass=&NiagaraSpriteRendererProperties, rendererIndex=1}
   {rendererClass=&NiagaraRibbonRendererProperties, rendererIndex=2}
   {rendererClass=&NiagaraRibbonRendererProperties, rendererIndex=3}
   {rendererClass=&NiagaraRibbonRendererProperties, rendererIndex=4}
  ]
 }
 LegRibbons002 {
  emitterUpdateScript [
   &EmitterState()
   /SetVariables(Emitter.RibbonLengthModifier:NiagaraFloat=&Multiply_Float(A=2, B=$User.RibbonLengthFactor)) {isSetParameters=true, module=SetVariables_#13}
   &InterpolateOverTime(
    Rate Of Change:NiagaraFloat=&MakeCustomFloatFromBool(Bool=&SetBoolByFloatComparison(A=$System.EngineOwnerSpeed, B=0.5), True Float=1, False Float=15)
    Target Float Value:NiagaraFloat=$System.EngineOwnerSpeed
   )
  ]
  particleSpawnScript [
   /SetVariables(Particles.MaxRibbonWidth=15) {isSetParameters=true, module=SetVariables_#14}
   &InitializeRibbon(
    Color=&MultiplyLinearColor(
     Linear Color=&HueShiftLinearColor(Hue=0, Saturation=0.9, Value=2, Linear Color=$User.BariumPaintColor)
     Scale Factor=<1 1 1 0.05>
    )
    Lifetime=0.2
    Ribbon Width=&FloatFromCurve(
     FloatCurve=DataInterface() {propertyValues {Curve {keys [
         {arriveTangent=4.1785, interpMode=RCIM_Cubic, leaveTangent=4.1785, tangentMode=RCTM_User, time=0, value=0}
         {arriveTangent=0.9995, interpMode=RCIM_Cubic, leaveTangent=0.9995, tangentMode=RCTM_User, time=0.2223, value=0.6825}
         {arriveTangent=-0.0405, interpMode=RCIM_Cubic, leaveTangent=-0.0405, tangentMode=RCTM_User, time=1, value=1}
        ]}}}
     CurveIndex=&ClampFloat(Float=&Divide_Float(A=$Emitter.InterpolateOverTime.InterpolateOverTime_Float.Moving Average, B=100))
     Scale Curve=$Particles.MaxRibbonWidth
    )
    UsePositionOffset=false
    Write Material Random=true
   )
   /SetVariables(Particles.EventPosition:NiagaraPosition=$Particles.Position) {isSetParameters=true, module=SetVariables_#15}
   /SetVariables(Particles.SpriteSize=<5 5>) {isSetParameters=true, module=SetVariables_#16}
   /SetVariables(Particles.MaxRibbonLength:NiagaraFloat=&Multiply_Float(A=1, B=$Emitter.RibbonLengthModifier)) {isSetParameters=true, module=SetVariables_#17}
   /SetVariables() {isSetParameters=true, module=SetVariables_#18}
   /SetVariables() {isSetParameters=true, module=SetVariables_#19}
   /SetVariables() {isSetParameters=true, module=SetVariables_#20}
  ]
  particleUpdateScript [
   &ParticleState()
   &Drag(Drag=50, Use Linear Drag=true, Use Rotational Drag=false)
   &PointAttractionForce(
    Attraction Radius=800
    Attraction Strength=&Multiply_Float(
     A=&Lerp_Float(A=200, B=0, Alpha=$System.NormalizedPlayerSpeed)
     B=&ClampFloat(Float=&Divide_Float(A=$Emitter.InterpolateOverTime.InterpolateOverTime_Float.Moving Average, B=100))
    )
    Attractor Position Offset:Vector3f=<0 0 100>
    Attractor Position Offset Coordinate Space:ENiagaraCoordinateSpace=::Local
    Kill Within Radius=false
    Module.Use Falloff=false
    UseAttractorOffset=true
   ) {enabled=false}
   &SolveForcesAndVelocity(Clamp Velocity=false, Limit Acceleration=false)
   /SetVariables(Particles.NormalizedDistanceFromParent:NiagaraFloat=&ClampFloat(Float=&Divide_Float(
      A=&VectorLength(
       Vector=&Subtract_Vector(
        A=Linked() {linkedVariable {name=Particles.Position, type=NiagaraPosition}}
        B=&self:GetParentParticlePosition(PAR:NiagaraDataInterfaceParticleRead=DataInterface() {propertyValues {EmitterBinding {bindingMode=Other, emitterName=RibbonSourcesLegs}}})
       )
       Fallback Vector=<0 0 0.01>
       Fallback Length=0
      )
      B=$Particles.MaxRibbonLength
     ))) {isSetParameters=true, module=SetVariables_#21}
   &ScaleColor(
    Scale Alpha=&MakeCustomFloatFromBool(
     Bool=$Particles.EventReceieved
     True Float=&FloatFromCurve(FloatCurve=DataInterface() {propertyValues {Curve {keys [<at 0 1>,<at 0.8 0>]}}}, CurveIndex=$Particles.NormalizedDistanceFromParent)
     False Float=1
    )
    Scale RGB=&VectorFromFloat(Value=&FloatFromCurve(
      FloatCurve=DataInterface() {propertyValues {Curve {keys [
          {arriveTangent=-2.725, interpMode=RCIM_Cubic, leaveTangent=-2.725, tangentMode=RCTM_User, time=0, value=5}
          {arriveTangent=-8.8847, leaveTangent=-8.8847, time=0.2, value=1}
          <at 1 0>
         ]}}}
      CurveIndex=$Particles.NormalizedDistanceFromParent
     ))
    ScaleA=true
    ScaleRGB=true
   )
   &ScaleColor(
    Scale Alpha=&FloatFromCurve(FloatCurve=DataInterface() {propertyValues {Curve {keys [
         {arriveTangent=-3.8552, interpMode=RCIM_Cubic, leaveTangent=-3.8552, tangentMode=RCTM_User, time=0, value=1}
         <at 0.5 0.1>
         {interpMode=RCIM_Cubic, tangentMode=RCTM_User, time=1, value=0.02}
        ]}}})
    Scale RGB=&VectorFromFloat(Value=&FloatFromCurve(
      FloatCurve=DataInterface() {propertyValues {Curve {keys [<at 0 1>,<at 0.1 0.1>]}}}
      CurveIndex=$Particles.NormalizedDistanceFromParent
      Scale Curve=0.4
     ))
    ScaleA=true
    ScaleRGB=true
   ) {module=ScaleColor002}
   /SetVariables(Particles.Color2=&MultiplyLinearColor(
     Linear Color=$User.BariumPaintColor
     Scale Factor=&MakeFloat4FromVectorAndFloat(
      "Vector (RGB)"=&VectorFromFloat(Value=&FloatFromCurve(
        FloatCurve=DataInterface() {propertyValues {Curve {keys [<at 0 1>,<at 0.15 0.1>]}}}
        CurveIndex=$Particles.NormalizedDistanceFromParent
        Scale Curve=0.2
       ))
      "Float (Alpha)"=&Multiply_Float(
       A=&FloatFromCurve(FloatCurve=DataInterface() {propertyValues {Curve {keys [
            {arriveTangent=-3.2069, interpMode=RCIM_Cubic, leaveTangent=-3.2069, tangentMode=RCTM_User, time=0, value=1}
            {interpMode=RCIM_Cubic, tangentMode=RCTM_User, time=1, value=0}
           ]}}})
       B=&FloatFromCurve(
        FloatCurve=DataInterface() {propertyValues {Curve {keys [
            {arriveTangent=-3.6157, interpMode=RCIM_Cubic, leaveTangent=-3.6157, tangentMode=RCTM_User, time=0, value=1}
            <at 0.1 0.1>
            {interpMode=RCIM_Cubic, tangentMode=RCTM_User, time=1, value=0}
           ]}}}
        CurveIndex=$System.NormalizedPlayerSpeed
       )
      )
     )
    )) {isSetParameters=true}
   &KillParticles(Kill Particles=&Boolean_AndOperation(
     Bool1=&SetBoolByFloatComparison(
      A=&MakeFloatFromLinearColor(LinearColor=$Particles.Color, Channel:ENiagaraLinearColor_Channels=::NewEnumerator3)
      B=0.02
      Comparison Type=::NewEnumerator6
     )
     Bool2=$Particles.EventReceieved
    )) {enabled=false, module=KillParticles002}
   /SetVariables(Particles.RibbonU0Override=&MakeCustomFloatFromBool(Bool=$Particles.EventReceieved, True Float=$Particles.NormalizedDistanceFromParent, False Float=0)) {isSetParameters=true, module=SetVariables_#22}
   &ScaleRibbonWidth(Ribbon Width Scale=&FloatFromCurve(
     FloatCurve=DataInterface() {propertyValues {Curve {keys [{interpMode=RCIM_Cubic, tangentMode=RCTM_User, time=0, value=0}, {interpMode=RCIM_Cubic, tangentMode=RCTM_User, time=1, value=1}]}}}
     CurveIndex=&MakeCustomFloatFromBool(
      Bool=$User.UseCinematicVelocity
      True Float=1
      False Float=&ClampFloat(Float=&Divide_Float(A=$Emitter.InterpolateOverTime.InterpolateOverTime_Float.Moving Average, B=100))
     )
     Scale Curve=&MakeCustomFloatFromBool(
      Bool=$Particles.EventReceieved
      True Float=&FloatFromCurve(
       FloatCurve=DataInterface() {propertyValues {Curve {keys [{tangentMode=RCTM_User, time=0, value=1}, {tangentMode=RCTM_User, time=1, value=1}]}}}
       CurveIndex=$Particles.NormalizedDistanceFromParent
      )
      False Float=1
     )
    ))
   &ScaleColor(
    Scale Alpha=&MakeCustomFloatFromBool(Bool=$User.bisFootHidden, True Float=0, False Float=1)
    ScaleA=true
    ScaleRGB=true
   ) {module=ScaleColor001}
  ]
  rendererClasses [&NiagaraRibbonRendererProperties, &NiagaraSpriteRendererProperties]
  renderers [
   {rendererClass=&NiagaraRibbonRendererProperties}
   {rendererClass=&NiagaraSpriteRendererProperties, rendererIndex=1}
   {rendererClass=&NiagaraRibbonRendererProperties, rendererIndex=2}
   {rendererClass=&NiagaraRibbonRendererProperties, rendererIndex=3}
  ]
 }
 RibbonSourcesHelmet {
  emitterSpawnScript [
   /SetVariables(Emitter.NumberOfRibbons=2) {isSetParameters=true, module=SetVariables_#23}
   /SetVariables(Emitter.BoneSocketArray=DataInterface() {propertyValues {GpuSyncMode=SyncCpuToGpu, MaxElements=0, IntData [20,21]}}) {isSetParameters=true, module=SetVariables_#24}
   /SetVariables(Emitter.MaxRibbonLengthArray=DataInterface() {propertyValues {GpuSyncMode=SyncCpuToGpu, MaxElements=0, FloatData [100,100]}}) {isSetParameters=true, module=SetVariables_#25}
  ]
  emitterUpdateScript [
   &EmitterState(
    Inactive Response=::NewEnumerator1
    Life Cycle Mode=::NewEnumerator1
    Loop Behavior=::NewEnumerator1
    Loop Duration Mode=::NewEnumerator1
    UseLoopDelay=false
   )
   &SpawnBurst_Instantaneous(Spawn Count=$Emitter.NumberOfRibbons, Use Spawn Probability=false, UseLoopCountLimit=false)
  ]
  particleSpawnScript [
   &FN_SetSpawnGroup(Normalize Spawn Group=true, Number of Spawn Groups:NiagaraInt32=$Emitter.NumberOfRibbons, Set Ribbon ID and Ribbon Link Order=true)
   &SkeletalMeshLocation(
    Bone / Socket Index=&SelectIntFromArray(Direct Array Index=$Particles.SpawnGroup, Int Selection Array=$Emitter.BoneSocketArray)
    MaskSpawnGroup=false
    Mesh Sampling Type=::NewEnumerator5
    Skeletal Mesh=$User.SkeletalMesh
    Socket Sampling Mode=::NewEnumerator5
   )
   /SetVariables() {isSetParameters=true, module=SetVariables_#26}
   /SetVariables(Particles.MaterialRandom:NiagaraFloat=&RandomRangeFloat(Minimum=0, Maximum=1)) {isSetParameters=true, module=SetVariables_#27}
  ]
  particleUpdateScript [
   &ParticleState(Kill Particles When Lifetime Has Elapsed=false, Let Infinitely Lived Particles Die When Emitter Deactivates=true)
   &SkeletalMeshLocation(
    Bone / Socket Index=&SelectIntFromArray(Direct Array Index=$Particles.SpawnGroup, Int Selection Array=$Emitter.BoneSocketArray)
    MaskSpawnGroup=false
    Mesh Sampling Type=::NewEnumerator5
    Skeletal Mesh=$User.SkeletalMesh
    Socket Sampling Mode=::NewEnumerator5
   ) {module=SkeletalMeshLocation001}
   /SetVariables(Particles.BoneXAxis:Vector3f=&ReturnMeshOrientationAxis(Mesh Orientation Axis=::XAxis, Mesh Orientation Quat=$Particles.SkeletalMeshLocation001.SampledRotation)) {isSetParameters=true, module=SetVariables_#28}
   &SolveForcesAndVelocity(Clamp Velocity=false, Limit Acceleration=false)
   /SetVariables(Particles.MaxRibbonLength:NiagaraFloat=&SelectFloatFromArray(Direct Array Index=$Particles.SpawnGroup, Float Selection Array=$Emitter.MaxRibbonLengthArray)) {isSetParameters=true, module=SetVariables_#29}
   &GenerateLocationEvent(
    Event Generation Enabled:NiagaraBool=&Boolean_AndOperation(
     Bool1=&Boolean_OrOperation(Bool1=&SetBoolByFloatComparison(A=$System.EngineOwnerSpeed, B=5), Bool2=$User.UseCinematicVelocity)
     Bool2=&Boolean_NotOperation(Bool=$User.bisHeadHidden)
    )
    Event Type:ENiagara_LocEventType=::NewEnumerator1
    "Float to Send as Float 1 (NormalizedAge)":NiagaraFloat=$Particles.MaxRibbonLength
    Per Unit Velocity Vector:Vector3f=&MakeCustomVectorFromBool(
     Bool=$User.UseCinematicVelocity
     True Vector=&Add_Vector(A=$User.CinematicVelocity, B=$Particles.SkeletalMeshLocation001.SampledVelocity)
     False Vector=$Particles.SkeletalMeshLocation001.SampledVelocity
    )
    Unit Spacing=8
    Use Event Probability=false
    UseEventDelay=false
    "Vector to Send as Vector 2 (Velocity)":Vector3f=$Particles.BoneXAxis
   ) {module=GenerateLocationEvent001}
  ]
  rendererClasses [&NiagaraSpriteRendererProperties]
  renderers [{rendererClass=&NiagaraSpriteRendererProperties}]
 }
 RibbonSourcesJacket {
  emitterSpawnScript [
   /SetVariables(Emitter.NumberOfRibbons=6) {isSetParameters=true, module=SetVariables_#23}
   /SetVariables(Emitter.BoneSocketArray=DataInterface() {propertyValues {GpuSyncMode=SyncCpuToGpu, MaxElements=0, IntData [2,14,6,3,15,7]}}) {isSetParameters=true, module=SetVariables_#24}
   /SetVariables(Emitter.BoneOffsetArray=DataInterface() {propertyValues {GpuSyncMode=SyncCpuToGpu, MaxElements=0, FloatData [<0 0 0>, <0 0 0>, <0 0 0>, <0 0 0>, <0 0 0>, <0 0 0>]}}) {isSetParameters=true, module=SetVariables_#30}
   /SetVariables(Emitter.MaxRibbonLengthArray=DataInterface() {propertyValues {GpuSyncMode=SyncCpuToGpu, MaxElements=0, FloatData [20,20,20,20,20,20]}}) {isSetParameters=true, module=SetVariables_#25}
  ]
  emitterUpdateScript [
   &EmitterState(
    Inactive Response=::NewEnumerator1
    Life Cycle Mode=::NewEnumerator1
    Loop Behavior=::NewEnumerator1
    Loop Duration Mode=::NewEnumerator1
    UseLoopDelay=false
   )
   &SpawnBurst_Instantaneous(Spawn Count=$Emitter.NumberOfRibbons, Use Spawn Probability=false, UseLoopCountLimit=false)
  ]
  particleSpawnScript [
   &FN_SetSpawnGroup(Normalize Spawn Group=true, Number of Spawn Groups:NiagaraInt32=$Emitter.NumberOfRibbons, Set Ribbon ID and Ribbon Link Order=true)
   &SkeletalMeshLocation(
    Bone / Socket Index=$Particles.SpawnGroup
    MaskSpawnGroup=false
    Mesh Sampling Type=::NewEnumerator5
    Sampled Position Offset=&SelectVectorFromArray(Direct Array Index=$Particles.SpawnGroup, Vector Selection Array=$Emitter.BoneOffsetArray)
    Skeletal Mesh=$User.SkeletalMesh
    Socket Sampling Mode=::NewEnumerator5
   )
   /SetVariables(Particles.SpriteSize=<5 5>) {isSetParameters=true, module=SetVariables_#26}
   /SetVariables(Particles.MaterialRandom:NiagaraFloat=&RandomRangeFloat(Minimum=0, Maximum=1)) {isSetParameters=true, module=SetVariables_#27}
  ]
  particleUpdateScript [
   &ParticleState(Kill Particles When Lifetime Has Elapsed=false, Let Infinitely Lived Particles Die When Emitter Deactivates=true)
   &SkeletalMeshLocation(
    Bone / Socket Index=&SelectIntFromArray(Direct Array Index=$Particles.SpawnGroup, Int Selection Array=$Emitter.BoneSocketArray)
    MaskSpawnGroup=false
    Mesh Sampling Type=::NewEnumerator5
    Sampled Position Offset=&SelectVectorFromArray(Direct Array Index=$Particles.SpawnGroup, Vector Selection Array=$Emitter.BoneOffsetArray)
    Skeletal Mesh=$User.SkeletalMesh
    Socket Sampling Mode=::NewEnumerator5
   ) {module=SkeletalMeshLocation001}
   /SetVariables(Particles.BoneXAxis:Vector3f=&ReturnMeshOrientationAxis(Mesh Orientation Axis=::XAxis, Mesh Orientation Quat=$Particles.SkeletalMeshLocation001.SampledRotation)) {isSetParameters=true, module=SetVariables_#28}
   &SolveForcesAndVelocity(Clamp Velocity=false, Limit Acceleration=false)
   /SetVariables(Particles.MaxRibbonLength:NiagaraFloat=&SelectFloatFromArray(Direct Array Index=$Particles.SpawnGroup, Float Selection Array=$Emitter.MaxRibbonLengthArray)) {isSetParameters=true, module=SetVariables_#29}
   &GenerateLocationEvent(
    Event Generation Enabled:NiagaraBool=&Boolean_OrOperation(Bool1=&SetBoolByFloatComparison(A=$System.EngineOwnerSpeed, B=2), Bool2=$User.UseCinematicVelocity)
    Event Type:ENiagara_LocEventType=::NewEnumerator1
    "Float to Send as Float 1 (NormalizedAge)":NiagaraFloat=$Particles.MaxRibbonLength
    Per Unit Velocity Vector:Vector3f=&MakeCustomVectorFromBool(
     Bool=$User.UseCinematicVelocity
     True Vector=&Add_Vector(A=$User.CinematicVelocity, B=$Particles.SkeletalMeshLocation001.SampledVelocity)
     False Vector=$Particles.SkeletalMeshLocation001.SampledVelocity
    )
    Unit Spacing=5
    Use Event Probability=false
    UseEventDelay=false
    "Vector to Send as Vector 2 (Velocity)":Vector3f=$Particles.BoneXAxis
   ) {module=GenerateLocationEvent001}
  ]
  rendererClasses [&NiagaraSpriteRendererProperties]
  renderers [{rendererClass=&NiagaraSpriteRendererProperties}]
 }
 RibbonSourcesLegs {
  emitterSpawnScript [
   /SetVariables(Emitter.NumberOfRibbons=6) {isSetParameters=true, module=SetVariables_#23}
   /SetVariables(Emitter.BoneSocketArray=DataInterface() {propertyValues {GpuSyncMode=SyncCpuToGpu, MaxElements=0, IntData [12,16,18,13,17,19]}}) {isSetParameters=true, module=SetVariables_#24}
   /SetVariables(Emitter.BoneOffsetArray=DataInterface() {propertyValues {GpuSyncMode=SyncCpuToGpu, MaxElements=0, FloatData [<0 0 0>, <0 0 0>, <0 0 0>, <0 0 0>, <0 0 0>, <0 0 0>]}}) {isSetParameters=true, module=SetVariables_#30}
   /SetVariables(Emitter.MaxRibbonLengthArray=DataInterface() {propertyValues {GpuSyncMode=SyncCpuToGpu, MaxElements=0, FloatData [200,200,200,200,200,200,200]}}) {isSetParameters=true, module=SetVariables_#25}
  ]
  emitterUpdateScript [
   &EmitterState(
    Inactive Response=::NewEnumerator1
    Life Cycle Mode=::NewEnumerator1
    Loop Behavior=::NewEnumerator1
    Loop Duration Mode=::NewEnumerator1
    UseLoopDelay=false
   )
   &SpawnBurst_Instantaneous(Spawn Count=$Emitter.NumberOfRibbons, Use Spawn Probability=false, UseLoopCountLimit=false)
  ]
  particleSpawnScript [
   &FN_SetSpawnGroup(Normalize Spawn Group=true, Number of Spawn Groups:NiagaraInt32=$Emitter.NumberOfRibbons, Set Ribbon ID and Ribbon Link Order=true)
   &SkeletalMeshLocation(
    Bone / Socket Index=$Particles.SpawnGroup
    MaskSpawnGroup=false
    Mesh Sampling Type=::NewEnumerator5
    Sampled Position Offset=&SelectVectorFromArray(Direct Array Index=$Particles.SpawnGroup, Vector Selection Array=$Emitter.BoneOffsetArray)
    Skeletal Mesh=$User.SkeletalMesh
    Socket Sampling Mode=::NewEnumerator5
   )
   /SetVariables(Particles.SpriteSize=<5 5>) {isSetParameters=true, module=SetVariables_#26}
   /SetVariables(Particles.MaterialRandom:NiagaraFloat=&RandomRangeFloat(Minimum=0, Maximum=1)) {isSetParameters=true, module=SetVariables_#27}
  ]
  particleUpdateScript [
   &ParticleState(Kill Particles When Lifetime Has Elapsed=false, Let Infinitely Lived Particles Die When Emitter Deactivates=true)
   &SkeletalMeshLocation(
    Bone / Socket Index=&SelectIntFromArray(Direct Array Index=$Particles.SpawnGroup, Int Selection Array=$Emitter.BoneSocketArray)
    MaskSpawnGroup=false
    Mesh Sampling Type=::NewEnumerator5
    Sampled Position Offset=&SelectVectorFromArray(Direct Array Index=$Particles.SpawnGroup, Vector Selection Array=$Emitter.BoneOffsetArray)
    Skeletal Mesh=$User.SkeletalMesh
    Socket Sampling Mode=::NewEnumerator5
   ) {module=SkeletalMeshLocation001}
   /SetVariables(Particles.BoneXAxis:Vector3f=&ReturnMeshOrientationAxis(Mesh Orientation Axis=::XAxis, Mesh Orientation Quat=$Particles.SkeletalMeshLocation001.SampledRotation)) {isSetParameters=true, module=SetVariables_#28}
   &SolveForcesAndVelocity(Clamp Velocity=false, Limit Acceleration=false)
   /SetVariables(Particles.MaxRibbonLength:NiagaraFloat=&SelectFloatFromArray(Direct Array Index=$Particles.SpawnGroup, Float Selection Array=$Emitter.MaxRibbonLengthArray)) {isSetParameters=true, module=SetVariables_#29}
   &GenerateLocationEvent(
    Event Generation Enabled:NiagaraBool=&Boolean_OrOperation(Bool1=&SetBoolByFloatComparison(A=$System.EngineOwnerSpeed, B=5), Bool2=$User.UseCinematicVelocity)
    Event Type:ENiagara_LocEventType=::NewEnumerator1
    "Float to Send as Float 1 (NormalizedAge)":NiagaraFloat=$Particles.MaxRibbonLength
    Per Unit Velocity Vector:Vector3f=&MakeCustomVectorFromBool(
     Bool=$User.UseCinematicVelocity
     True Vector=&Add_Vector(A=$User.CinematicVelocity, B=$Particles.SkeletalMeshLocation001.SampledVelocity)
     False Vector=$Particles.SkeletalMeshLocation001.SampledVelocity
    )
    Unit Spacing=5
    Use Event Probability=false
    UseEventDelay=false
    "Vector to Send as Vector 2 (Velocity)":Vector3f=$Particles.BoneXAxis
   ) {module=GenerateLocationEvent001}
  ]
  rendererClasses [&NiagaraSpriteRendererProperties]
  renderers [{rendererClass=&NiagaraSpriteRendererProperties}]
 }
 Speed_Line_Helmet {
  emitterSpawnScript [
   /SetVariables(Emitter.EmitRefract=1.5) {isSetParameters=true, module=SetVariables_#31}
   /SetVariables(Emitter.EmitColorLightning:LinearColor=$User.BariumPaintColor) {isSetParameters=true, module=SetVariables_#32}
  ]
  emitterUpdateScript [
   &EmitterState()
   &SpawnRate(SpawnRate=&Multiply_Float(A=$System.NormalizedSpeedMask, B=20), Use Spawn Probability=false) {enabled=false}
   &SpawnPerUnit(
    Max Movement Threshold=900
    Module.Use Max Movement Threshold=true
    Module.Use Movement Tolerance=true
    Spawn Probability:NiagaraFloat=&MakeCustomFloatFromBool(Bool=$User.bisHeadHidden, True Float=0, False Float=$User.SpawnHelmetFX)
    Spawn Spacing=70
    Use Spawn Probability=true
   )
  ]
  particleSpawnScript [
   &InitializeParticle(
    Lifetime Max=0.2
    Lifetime Min=0.1
    Lifetime Mode=::NewEnumerator1
    Sprite Size=&MakeVector2D(
     X=&RandomRangeFloat(Minimum=0.4, Maximum=0.7)
     Y=&RandomRangeFloat(Minimum=0.4, Maximum=2)
    )
    Sprite Size Mode=::NewEnumerator4
    UsePositionOffset=false
    Write Material Random=true
   )
   &SkeletalMeshLocation(
    Bone / Socket Index=&RandomRangeInt(Minimum=0, Maximum=1)
    "Inherit Velocity (Scale)"=0.1
    MaskSpawnGroup=false
    Mesh Sampling Type=::NewEnumerator5
    Sampled Position Offset=&RandomRangeVector(Minimum=<-5 4 5>, Maximum=<5 5 10>)
    Sampled Position Offset Coordinate Space:ENiagaraSkelMeshTransforms=::NewEnumerator2
    Skeletal Mesh=$User.SkeletalMesh
    Socket Sampling Mode=::NewEnumerator5
    Velocity Sampling=::NewEnumerator0
    VelocitySpeedLimit=true
   )
   &InheritVelocity(Inherited Velocity Speed Limit=300, Speed Limit=true, UseSpeedThreshold=false)
   &ApplyInitialForces()
  ]
  particleUpdateScript [
   &ParticleState()
   &SkeletalMeshLocation(
    Bone / Socket Index=&RandomRangeInt(Minimum=2, Maximum=17)
    MaskSpawnGroup=false
    Mesh Sampling Type=::NewEnumerator5
    Position Sampling=::NewEnumerator2
    Skeletal Mesh=$User.SkeletalMesh
    Socket Sampling Mode=::NewEnumerator5
   ) {enabled=false, module=SkeletalMeshLocation001}
   &GravityForce(Gravity=<0 0 -100>)
   &SolveForcesAndVelocity(Clamp Velocity=false, Limit Acceleration=false)
   &Color(Color=&MakeLinearColorFromVectorAndFloat(
     "Vector (RGB)"=&VectorFromFloat(Value=&FloatFromCurve(FloatCurve=DataInterface() {propertyValues {Curve {keys [<at 0 1>,<at 1 0>]}}}))
     "Float (Alpha)"=1
    ))
   /SetVariables(Particles.Color2=&HueShiftLinearColor(Hue=0, Saturation=1, Value=2, Linear Color=$User.BariumPaintColor)) {isSetParameters=true, module=SetVariables_#33}
   &ScaleColor(
    Scale Alpha=&Lerp_Float(
     A=0
     B=1
     Alpha=&ClampFloat(Float=&Multiply_Float(A=6, B=$System.NormalizedSpeedMask))
    )
    Scale RGB=&Multiply_Vector(
     A=<1 1 1>
     B=&Lerp_Vector(
      A=&MakeVectorFromLinearColorRGB(Color=$User.BariumPaintColor)
      B=<1 1 1>
      Alpha=$System.NormalizedPlayerSpeed
     )
    )
    ScaleA=true
    ScaleRGB=true
   ) {enabled=false}
   &ScaleColor(
    Scale Alpha=&FloatFromCurve(FloatCurve=DataInterface() {propertyValues {Curve {keys [
         {arriveTangent=9.0725, interpMode=RCIM_Cubic, leaveTangent=9.0725, tangentMode=RCTM_User, time=0, value=0}
         {arriveTangent=-2.5249, interpMode=RCIM_Cubic, leaveTangent=-2.5249, tangentMode=RCTM_User, time=0.1408, value=0.9847}
         <at 0.4 0.25>
         {arriveTangent=-0.0405, interpMode=RCIM_Cubic, leaveTangent=-0.0405, tangentMode=RCTM_User, time=1, value=0}
        ]}}})
    Scale RGB=&VectorFromFloat(Value=&FloatFromCurve(
      FloatCurve=DataInterface() {propertyValues {Curve {keys [
          {arriveTangent=-2.8498, interpMode=RCIM_Cubic, leaveTangent=-2.8498, tangentMode=RCTM_User, time=0, value=1}
          {arriveTangent=-1, leaveTangent=-1, time=0.9, value=0.05}
          <at 1 0 RCIM_Cubic>
         ]}}}
      Scale Curve=5
     ))
    ScaleA=true
    ScaleRGB=true
   ) {enabled=false, module=ScaleColor001}
   &CameraOffset(Camera Offset Amount=8)
   &DynamicMaterialParameters(
    Index 0 Param 1=&RandomRangeFloat(Minimum=0, Maximum=1)
    Index 0 Param 2=&Lerp_Float(
     A=0.9
     B=1
     Alpha=&ClampFloat(Float=&Multiply_Float(A=6, B=$System.NormalizedSpeedMask))
    )
    Module.Param0WriteEnabled=true
    Module.Param1WriteEnabled=true
    Module.Param2WriteEnabled=true
    Module.Param3WriteEnabled=true
   )
   &SpriteFacingAndAlignment(Sprite Alignment=$Engine.Owner.Velocity, WriteAlignment=true, WriteFacing=false)
   &ScaleSpriteSizeBySpeed(
    Max Scale Factor=<0.5 60>
    Min Scale Factor=<6 1>
    Sample Scale Factor By Curve=true
    Scale Factor Curve=DataInterface() {propertyValues {Curve {keys [
        {interpMode=RCIM_Cubic, tangentMode=RCTM_User, time=0, value=0}
        {arriveTangent=2.3236, interpMode=RCIM_Cubic, leaveTangent=2.3236, tangentMode=RCTM_User, time=1, value=1}
       ]}}}
    Velocity Threshold=539
   ) {enabled=false}
   &ScaleSpriteSize(
    Non-Uniform Curve Index:NiagaraFloat=$System.NormalizedPlayerSpeed
    Non-Uniform Curve Scale=<5 50>
    Non-Uniform Curve Sprite Scale=DataInterface() {propertyValues {
      XCurve {keys [
        {arriveTangent=-3.0556, interpMode=RCIM_Cubic, leaveTangent=-3.0556, tangentMode=RCTM_User, time=0, value=1}
        {interpMode=RCIM_Cubic, tangentMode=RCTM_User, time=1, value=0.2}
       ]}
      YCurve {keys [
        {arriveTangent=0.2804, interpMode=RCIM_Cubic, leaveTangent=0.2804, tangentMode=RCTM_User, time=0, value=0}
        {arriveTangent=1.4097, interpMode=RCIM_Cubic, leaveTangent=1.4097, tangentMode=RCTM_User, time=1, value=1}
       ]}
     }}
    Scale Sprite Size Mode=::NewEnumerator3
   )
  ]
  rendererClasses [&NiagaraSpriteRendererProperties]
  renderers [
   {rendererClass=&NiagaraSpriteRendererProperties}
   {rendererClass=&NiagaraSpriteRendererProperties, rendererIndex=1}
  ]
 }
 Speed_Line_Sprites {
  emitterSpawnScript [
   /SetVariables(Emitter.EmitRefract=1.5) {isSetParameters=true, module=SetVariables_#31}
   /SetVariables(Emitter.EmitColorLightning:LinearColor=$User.BariumPaintColor) {isSetParameters=true, module=SetVariables_#32}
  ]
  emitterUpdateScript [
   &EmitterState()
   &SpawnRate(SpawnRate=&Multiply_Float(A=$System.NormalizedSpeedMask, B=20), Use Spawn Probability=false) {enabled=false}
   &SpawnPerUnit(
    Max Movement Threshold=900
    Module.Use Max Movement Threshold=true
    Module.Use Movement Tolerance=true
    Spawn Probability:NiagaraFloat=$System.NormalizedPlayerSpeed
    Use Spawn Probability=true
   )
  ]
  particleSpawnScript [
   &InitializeParticle(
    Lifetime Max=0.4
    Lifetime Min=0.1
    Lifetime Mode=::NewEnumerator1
    Sprite Size=&MakeVector2D(
     X=&RandomRangeFloat(Minimum=0.4, Maximum=0.6)
     Y=&RandomRangeFloat(Minimum=0.4, Maximum=3)
    )
    Sprite Size Mode=::NewEnumerator4
    UsePositionOffset=false
    Write Material Random=true
   )
   &SkeletalMeshLocation(
    Bone / Socket Index=&RandomRangeInt(Minimum=2, Maximum=7)
    "Inherit Velocity (Scale)"=0.1
    MaskSpawnGroup=false
    Mesh Sampling Type=::NewEnumerator5
    Skeletal Mesh=$User.SkeletalMesh
    Socket Sampling Mode=::NewEnumerator5
    Velocity Sampling=::NewEnumerator0
    VelocitySpeedLimit=true
   )
   &InheritVelocity(Inherited Velocity Speed Limit=300, Speed Limit=true, UseSpeedThreshold=false)
   &ApplyInitialForces()
  ]
  particleUpdateScript [
   &ParticleState()
   &SkeletalMeshLocation(
    Bone / Socket Index=&RandomRangeInt(Minimum=2, Maximum=17)
    MaskSpawnGroup=false
    Mesh Sampling Type=::NewEnumerator5
    Position Sampling=::NewEnumerator2
    Skeletal Mesh=$User.SkeletalMesh
    Socket Sampling Mode=::NewEnumerator5
   ) {module=SkeletalMeshLocation001}
   &GravityForce(Gravity=<0 0 -100>)
   &SolveForcesAndVelocity(Clamp Velocity=false, Limit Acceleration=false)
   &Color(Color=&MakeLinearColorFromVectorAndFloat(
     "Vector (RGB)"=&VectorFromFloat(Value=&FloatFromCurve(FloatCurve=DataInterface() {propertyValues {Curve {keys [<at 0 1>,<at 1 0>]}}}))
     "Float (Alpha)"=1
    ))
   /SetVariables(Particles.Color2=&HueShiftLinearColor(Hue=0, Saturation=1, Value=2, Linear Color=$User.BariumPaintColor)) {isSetParameters=true, module=SetVariables_#33}
   &ScaleColor(
    Scale Alpha=&Lerp_Float(
     A=0
     B=1
     Alpha=&ClampFloat(Float=&Multiply_Float(A=6, B=$System.NormalizedSpeedMask))
    )
    Scale RGB=&Multiply_Vector(
     A=<1 1 1>
     B=&Lerp_Vector(
      A=&MakeVectorFromLinearColorRGB(Color=$User.BariumPaintColor)
      B=<1 1 1>
      Alpha=$System.NormalizedPlayerSpeed
     )
    )
    ScaleA=true
    ScaleRGB=true
   ) {enabled=false}
   &ScaleColor(
    Scale Alpha=&FloatFromCurve(FloatCurve=DataInterface() {propertyValues {Curve {keys [
         {arriveTangent=9.0725, interpMode=RCIM_Cubic, leaveTangent=9.0725, tangentMode=RCTM_User, time=0, value=0}
         {arriveTangent=-2.5249, interpMode=RCIM_Cubic, leaveTangent=-2.5249, tangentMode=RCTM_User, time=0.1408, value=0.9847}
         <at 0.4 0.25>
         {arriveTangent=-0.0405, interpMode=RCIM_Cubic, leaveTangent=-0.0405, tangentMode=RCTM_User, time=1, value=0}
        ]}}})
    Scale RGB=&VectorFromFloat(Value=&FloatFromCurve(
      FloatCurve=DataInterface() {propertyValues {Curve {keys [
          {arriveTangent=-2.8498, interpMode=RCIM_Cubic, leaveTangent=-2.8498, tangentMode=RCTM_User, time=0, value=1}
          {arriveTangent=-1, leaveTangent=-1, time=0.9, value=0.05}
          <at 1 0 RCIM_Cubic>
         ]}}}
      Scale Curve=5
     ))
    ScaleA=true
    ScaleRGB=true
   ) {enabled=false, module=ScaleColor001}
   &CameraOffset(Camera Offset Amount=8)
   &DynamicMaterialParameters(
    Index 0 Param 1=&RandomRangeFloat(Minimum=0, Maximum=1)
    Index 0 Param 2=&Lerp_Float(
     A=0.9
     B=1
     Alpha=&ClampFloat(Float=&Multiply_Float(A=6, B=$System.NormalizedSpeedMask))
    )
    Module.Param0WriteEnabled=true
    Module.Param1WriteEnabled=true
    Module.Param2WriteEnabled=true
    Module.Param3WriteEnabled=true
   )
   &SpriteFacingAndAlignment(Sprite Alignment=$Engine.Owner.Velocity, WriteAlignment=true, WriteFacing=false)
   &ScaleSpriteSizeBySpeed(
    Max Scale Factor=<0.5 60>
    Min Scale Factor=<6 1>
    Sample Scale Factor By Curve=true
    Scale Factor Curve=DataInterface() {propertyValues {Curve {keys [
        {interpMode=RCIM_Cubic, tangentMode=RCTM_User, time=0, value=0}
        {arriveTangent=2.3236, interpMode=RCIM_Cubic, leaveTangent=2.3236, tangentMode=RCTM_User, time=1, value=1}
       ]}}}
    Velocity Threshold=539
   ) {enabled=false}
   &ScaleSpriteSize(
    Non-Uniform Curve Index:NiagaraFloat=$System.NormalizedPlayerSpeed
    Non-Uniform Curve Scale=<5 50>
    Non-Uniform Curve Sprite Scale=DataInterface() {propertyValues {
      XCurve {keys [
        {arriveTangent=-3.0556, interpMode=RCIM_Cubic, leaveTangent=-3.0556, tangentMode=RCTM_User, time=0, value=1}
        {interpMode=RCIM_Cubic, tangentMode=RCTM_User, time=1, value=0.2}
       ]}
      YCurve {keys [
        {arriveTangent=0.2804, interpMode=RCIM_Cubic, leaveTangent=0.2804, tangentMode=RCTM_User, time=0, value=0}
        {arriveTangent=1.4097, interpMode=RCIM_Cubic, leaveTangent=1.4097, tangentMode=RCTM_User, time=1, value=1}
       ]}
     }}
    Scale Sprite Size Mode=::NewEnumerator3
   )
  ]
  rendererClasses [&NiagaraSpriteRendererProperties]
  renderers [
   {rendererClass=&NiagaraSpriteRendererProperties}
   {rendererClass=&NiagaraSpriteRendererProperties, rendererIndex=1}
  ]
 }
 Speed_Line_Sprites_Fat {
  emitterSpawnScript [
   /SetVariables(Emitter.EmitRefract=1.5) {isSetParameters=true, module=SetVariables_#31}
   /SetVariables(Emitter.EmitColorLightning:LinearColor=$User.BariumPaintColor) {isSetParameters=true, module=SetVariables_#32}
  ]
  emitterUpdateScript [
   &EmitterState()
   &SpawnRate(SpawnRate=&Multiply_Float(A=$System.NormalizedSpeedMask, B=20), Use Spawn Probability=false) {enabled=false}
   &SpawnPerUnit(
    Max Movement Threshold=6000
    Module.Use Max Movement Threshold=true
    Module.Use Movement Tolerance=true
    Spawn Probability:NiagaraFloat=$System.NormalizedPlayerSpeed
    Spawn Spacing=10
    Use Spawn Probability=true
   )
  ]
  particleSpawnScript [
   &InitializeParticle(
    Color=<rgba 1 1 1 0.03>
    Lifetime Max=0.4
    Lifetime Min=0.4
    Lifetime Mode=::NewEnumerator1
    Sprite Size=&MakeVector2D(X=&RandomRangeFloat(Minimum=0.2, Maximum=0.2), Y=1)
    Sprite Size Mode=::NewEnumerator4
    UsePositionOffset=false
    Write Material Random=true
   )
   &SkeletalMeshLocation(
    Bone / Socket Index=&RandomRangeInt(Minimum=8, Maximum=13)
    "Inherit Velocity (Scale)"=0.1
    MaskSpawnGroup=false
    Mesh Sampling Type=::NewEnumerator5
    Skeletal Mesh=$User.SkeletalMesh
    Socket Sampling Mode=::NewEnumerator5
    Velocity Sampling=::NewEnumerator0
    VelocitySpeedLimit=true
   )
   &InheritVelocity(Inherited Velocity Speed Limit=300, Speed Limit=true, UseSpeedThreshold=false)
   &ApplyInitialForces()
  ]
  particleUpdateScript [
   &ParticleState()
   &SkeletalMeshLocation(
    Bone / Socket Index=&RandomRangeInt(Minimum=0, Maximum=1)
    MaskSpawnGroup=false
    Mesh Sampling Type=::NewEnumerator5
    Position Sampling=::NewEnumerator2
    Skeletal Mesh=$User.SkeletalMesh
    Socket Sampling Mode=::NewEnumerator5
   ) {module=SkeletalMeshLocation001}
   &GravityForce(Gravity=<0 0 -100>)
   &SolveForcesAndVelocity(Clamp Velocity=false, Limit Acceleration=false)
   &ScaleColor(
    Scale Alpha=&FloatFromCurve(
     FloatCurve=DataInterface() {propertyValues {Curve {keys [{tangentMode=RCTM_User, time=0, value=0.5}, {tangentMode=RCTM_User, time=1, value=1}]}}}
     CurveIndex=$System.NormalizedPlayerSpeed
    )
    ScaleA=true
    ScaleRGB=true
   )
   &ScaleSpriteSizeBySpeed(
    Max Scale Factor=<30 100>
    Min Scale Factor=<30 20>
    Sample Scale Factor By Curve=true
    Scale Factor Curve=DataInterface() {propertyValues {Curve {keys [{interpMode=RCIM_Cubic, tangentMode=RCTM_User, time=0, value=0}, {interpMode=RCIM_Cubic, tangentMode=RCTM_User, time=1, value=1}]}}}
    Velocity Threshold=539
   )
   &ScaleColor(
    Scale Alpha=&FloatFromCurve(
     FloatCurve=DataInterface() {propertyValues {Curve {keys [
         {arriveTangent=9.0725, interpMode=RCIM_Cubic, leaveTangent=9.0725, tangentMode=RCTM_User, time=0, value=0}
         {arriveTangent=-2.5249, interpMode=RCIM_Cubic, leaveTangent=-2.5249, tangentMode=RCTM_User, time=0.3, value=0.9847}
         {arriveTangent=-0.0405, interpMode=RCIM_Cubic, leaveTangent=-0.0405, tangentMode=RCTM_User, time=1, value=0}
        ]}}}
     Scale Curve=0.5
    )
    Scale RGB=&VectorFromFloat(Value=&FloatFromCurve(
      FloatCurve=DataInterface() {propertyValues {Curve {keys [
          {arriveTangent=9.0725, interpMode=RCIM_Cubic, leaveTangent=9.0725, tangentMode=RCTM_User, time=0, value=0}
          {arriveTangent=-2.5249, interpMode=RCIM_Cubic, leaveTangent=-2.5249, tangentMode=RCTM_User, time=0.4, value=0.9847}
          {arriveTangent=-0.0405, interpMode=RCIM_Cubic, leaveTangent=-0.0405, tangentMode=RCTM_User, time=1, value=0}
         ]}}}
      Scale Curve=2
     ))
    ScaleA=true
    ScaleRGB=true
   ) {module=ScaleColor001}
   &CameraOffset(Camera Offset Amount=8)
   &DynamicMaterialParameters(
    Index 0 Param 1=&RandomRangeFloat(Minimum=0, Maximum=1)
    Index 0 Param 2=&Lerp_Float(
     A=0.9
     B=1
     Alpha=&ClampFloat(Float=&Multiply_Float(A=6, B=$System.NormalizedSpeedMask))
    )
    Module.Param0WriteEnabled=true
    Module.Param1WriteEnabled=true
    Module.Param2WriteEnabled=true
    Module.Param3WriteEnabled=true
   )
   &SpriteFacingAndAlignment(Sprite Alignment=$Engine.Owner.Velocity, WriteAlignment=true, WriteFacing=false)
   &ScaleSpriteSize(
    Scale Factor=&MakeVector2D(
     X=&Multiply_Float(
      A=&FloatFromCurve(
       FloatCurve=DataInterface() {propertyValues {Curve {keys [
           {arriveTangent=-3.3422, interpMode=RCIM_Cubic, leaveTangent=-3.3422, tangentMode=RCTM_User, time=0, value=1}
           {arriveTangent=-0.6568, interpMode=RCIM_Cubic, leaveTangent=-0.6568, tangentMode=RCTM_User, time=1, value=0}
          ]}}}
       Scale Curve=30
      )
      B=&FloatFromCurve(
       FloatCurve=DataInterface() {propertyValues {Curve {keys [{tangentMode=RCTM_User, time=0, value=0.5}, {tangentMode=RCTM_User, time=1, value=1}]}}}
       CurveIndex=&VectorLength(Vector=$Engine.Owner.Velocity)
      )
     )
     Y=&Lerp_Float(
      A=1
      B=6
      Alpha=&FloatFromCurve(
       FloatCurve=DataInterface() {propertyValues {Curve {keys [
           {interpMode=RCIM_Cubic, tangentMode=RCTM_User, time=0, value=0}
           {arriveTangent=2.4302, interpMode=RCIM_Cubic, leaveTangent=2.4302, tangentMode=RCTM_User, time=1, value=1}
          ]}}}
       CurveIndex=&VectorLength(Vector=$Engine.Owner.Velocity)
      )
     )
    )
    Scale Sprite Size Mode=::NewEnumerator2
   ) {enabled=false}
  ]
  rendererClasses [&NiagaraSpriteRendererProperties]
  renderers [{rendererClass=&NiagaraSpriteRendererProperties}]
 }
 TestReaction {
  bEnabled=false
  emitterUpdateScript [
   &EmitterState()
   &SpawnRate(Use Spawn Probability=false)
  ]
  particleSpawnScript [&InitializeParticle(Color Mode=::NewEnumerator0, Position Offset=<0 0 150>, UsePositionOffset=true, Write Material Random=true)]
  particleUpdateScript [
   &ParticleState()
   &DynamicMaterialParameters(
    Index 0 Param 1=$System.NormalizedSpeedMask
    Index 0 Param 2=$System.KillBoostFade
    Index 0 Param 3=$System.NormalizedOverallMask
    Index 0 Param 4=&VectorLength(Vector=$Engine.Owner.Velocity)
    Module.Param0WriteEnabled=true
    Module.Param1WriteEnabled=true
    Module.Param2WriteEnabled=true
    Module.Param3WriteEnabled=true
   )
   &CameraOffset(Camera Offset Amount=60)
  ]
  rendererClasses [&NiagaraSpriteRendererProperties]
  renderers [{rendererClass=&NiagaraSpriteRendererProperties}]
 }
}
summary {
 systemName=NS_Barium_Char_Idle
 emitters [
  {bEnabled=false, emitterName=TestReaction, rendererClasses [&NiagaraSpriteRendererProperties]}
  {emitterName=Speed_Line_Sprites, rendererClasses [&NiagaraSpriteRendererProperties]}
  {emitterName=RibbonSourcesJacket, rendererClasses [&NiagaraSpriteRendererProperties]}
  {emitterName=RibbonSourcesLegs, rendererClasses [&NiagaraSpriteRendererProperties]}
  {
   emitterName=HelmetRibbons
   rendererClasses [&NiagaraRibbonRendererProperties, &NiagaraSpriteRendererProperties]
  }
  {emitterName=RibbonSourcesHelmet, rendererClasses [&NiagaraSpriteRendererProperties]}
  {emitterName=Speed_Line_Sprites_Fat, rendererClasses [&NiagaraSpriteRendererProperties]}
  {
   emitterName=JacketRibbons002
   rendererClasses [&NiagaraRibbonRendererProperties, &NiagaraSpriteRendererProperties]
  }
  {
   emitterName=LegRibbons002
   rendererClasses [&NiagaraRibbonRendererProperties, &NiagaraSpriteRendererProperties]
  }
  {emitterName=Speed_Line_Helmet, rendererClasses [&NiagaraSpriteRendererProperties]}
  {emitterName=Decals_Shoulders, rendererClasses [&NiagaraSpriteRendererProperties]}
  {emitterName=Decals_Head, rendererClasses [&NiagaraSpriteRendererProperties]}
  {emitterName=Decals_Shoulders001, rendererClasses [&NiagaraSpriteRendererProperties]}
 ]
 userVariables [
  User.BariumPaintColor {type=LinearColor, description="", defaultValue {struct=LinearColor, value=<rgba 0.4793 0.5906 0.0331 1>}}
  User.bisFootHidden {type=NiagaraBool}
  User.bIsFrontEnd {type=NiagaraBool}
  User.bIsFrontEndPreview {type=NiagaraBool}
  User.bisHeadHidden {type=NiagaraBool}
  User.bisHelmet {type=NiagaraBool}
  User.CinematicVelocity {type=Vector3f, description="", defaultValue {struct=Vector3f, value=<0 200 200>}}
  User.Distance {type=NiagaraFloat, defaultValue=0, description=""}
  User.LightningAcceleration {type=NiagaraFloat, defaultValue=0, description=""}
  User.RibbonLengthFactor {type=NiagaraFloat, defaultValue=1, description=""}
  User.SkeletalMesh {
   type=NiagaraDataInterfaceSkeletalMesh
   defaultValue {
    struct=NiagaraExt_VariableValue_DataInterface
    value {dataInterface=&self:NiagaraDataInterfaceSkeletalMesh_0}
   }
  }
  User.SpawnHelmetFX {type=NiagaraFloat, defaultValue=0, description=""}
  User.SpawnJacketFX {type=NiagaraFloat, defaultValue=0, description=""}
  User.SpeedLine {type=NiagaraFloat, defaultValue=0, description=""}
  User.UseCinematicVelocity {type=NiagaraBool}
 ]
}
