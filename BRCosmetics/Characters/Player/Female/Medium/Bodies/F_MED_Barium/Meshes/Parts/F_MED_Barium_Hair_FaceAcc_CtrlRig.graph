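# Control Rig for the hair and face accessory part of the Barium medium female character, applying corrective rotations to master hair bones.
# While is_sky_diving is true, a shared fixed rotation is applied to the center and rear side hair masters.
# The head_tilt_right variable weights per-bone rotations on center, right and one left master.
# A spherical pose reader on upperarm_l, multiplied by that tilt, weights further rotations on center and right masters.
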ControlRigBlueprint /BRCosmetics/Characters/Player/Female/Medium/Bodies/F_MED_Barium/Meshes/Parts/F_MED_Barium_Hair_FaceAcc_CtrlRig
graphs [
 RigVMFunctionLibrary {links [], nodes []}
 RigVMModel {nodes [
   RigUnit_BeginExecution = &RigUnit_BeginExecution()
   n1 = /Script/ControlRig.RigUnit_SequenceExecution(ExecuteContext=@RigUnit_BeginExecution.ExecutePin) {is=struct RigVMUnitNode}
   Branch_RigVMUnitNode = &RigVMFunction_ControlFlowBranch(ExecuteContext=@n1.A, Condition=RigVMVariableNode() {variable=is_sky_diving}.Value)
   n2 = &RigUnit_SequenceExecution(ExecuteContext=@Branch_RigVMUnitNode.True)
   n3 = &RigVMFunction_MathQuaternionFromEuler(Euler=(X=0,Y=0,Z=-10))
   &RigUnit_ModifyTransforms(
    ItemToModify=((Item=(Type=Bone,Name="dyn_hair_04_mstr_l"),Transform=(Rotation=(X=0,Y=0,Z=0,W=1),Translation=(X=0,Y=0,Z=0),Scale3D=(X=1,Y=1,Z=1))))
    Weight=1
    ItemToModify.0.Transform.Rotation=@n3.Result
    ExecutePin=@n2.A
   )
   > &RigUnit_ModifyTransforms(
    ItemToModify=((Item=(Type=Bone,Name="dyn_hair_mstr_ct"),Transform=(Rotation=(X=0,Y=0,Z=0,W=1),Translation=(X=0,Y=0,Z=0),Scale3D=(X=1,Y=1,Z=1))))
    Weight=1
    ItemToModify.0.Transform.Rotation=@n3.Result
   )
   > &RigUnit_ModifyTransforms(
    ItemToModify=((Item=(Type=Bone,Name="dyn_hair_04_mstr_r"),Transform=(Rotation=(X=0,Y=0,Z=0,W=1),Translation=(X=0,Y=0,Z=0),Scale3D=(X=1,Y=1,Z=1))))
    Weight=1
    ItemToModify.0.Transform.Rotation=@n3.Result
   )
   &RigUnit_ModifyTransforms(
    ItemToModify=((Item=(Type=Bone,Name="dyn_hair_03_mstr_l"),Transform=(Rotation=(X=0,Y=0,Z=0,W=1),Translation=(X=0,Y=0,Z=0),Scale3D=(X=1,Y=1,Z=1))))
    Weight=1
    ItemToModify.0.Transform.Rotation=@n3.Result
    ExecutePin=@n2.B
   )
   > &RigUnit_ModifyTransforms(
    ItemToModify=((Item=(Type=Bone,Name="dyn_hair_03_mstr_r"),Transform=(Rotation=(X=0,Y=0,Z=0,W=1),Translation=(X=0,Y=0,Z=0),Scale3D=(X=1,Y=1,Z=1))))
    Weight=1
    ItemToModify.0.Transform.Rotation=@n3.Result
   )
   n4 = &RigUnit_SequenceExecution(ExecuteContext=@n1.B)
   n5 = RigVMVariableNode() {variable=head_tilt_right}
   &RigUnit_ModifyTransforms(
    ItemToModify=((Item=(Type=Bone,Name="dyn_hair_04_mstr_l"),Transform=(Rotation=(X=0,Y=0,Z=0,W=1),Translation=(X=0,Y=0,Z=0),Scale3D=(X=1,Y=1,Z=1))))
    ItemToModify.0.Transform.Rotation=&RigVMFunction_MathQuaternionFromEuler(Euler=(X=0,Y=10,Z=-20)).Result
    ExecutePin=@n4.A
    Weight=@n5.Value
   )
   > &RigUnit_ModifyTransforms(
    ItemToModify=((Item=(Type=Bone,Name="dyn_hair_mstr_ct"),Transform=(Rotation=(X=0,Y=0,Z=0,W=1),Translation=(X=0,Y=0,Z=0),Scale3D=(X=1,Y=1,Z=1))))
    ItemToModify.0.Transform.Rotation=&RigVMFunction_MathQuaternionFromEuler(Euler=(X=0,Y=10,Z=-30)).Result
    Weight=@n5.Value
   )
   > &RigUnit_ModifyTransforms(
    ItemToModify=((Item=(Type=Bone,Name="dyn_hair_04_mstr_r"),Transform=(Rotation=(X=0,Y=0,Z=0,W=1),Translation=(X=0,Y=0,Z=0),Scale3D=(X=1,Y=1,Z=1))))
    ItemToModify.0.Transform.Rotation=&RigVMFunction_MathQuaternionFromEuler(Euler=(X=0,Y=10,Z=-40)).Result
    Weight=@n5.Value
   )
   > &RigUnit_ModifyTransforms(
    ItemToModify=((Item=(Type=Bone,Name="dyn_hair_03_mstr_r"),Transform=(Rotation=(X=0,Y=0,Z=0,W=1),Translation=(X=0,Y=0,Z=0),Scale3D=(X=1,Y=1,Z=1))))
    ItemToModify.0.Transform.Rotation=&RigVMFunction_MathQuaternionFromEuler(Euler=(X=0,Y=-60,Z=-40)).Result
    Weight=@n5.Value
   )
   n6 = &RigUnit_SphericalPoseReader(
    ActiveRegionScaleFactors=(PositiveWidth=1,NegativeWidth=1,PositiveHeight=1,NegativeHeight=1)
    ActiveRegionSize=0.025
    Debug=(bDrawDebug=false,bDrawLocalAxes=true,DebugScale=25,DebugSegments=20,DebugThickness=0.25)
    DriverAxis=(X=1,Y=0,Z=0)
    DriverItem=(Type=Bone,Name="upperarm_l")
    FalloffRegionScaleFactors=(PositiveWidth=1,NegativeWidth=1,PositiveHeight=1,NegativeHeight=1)
    FalloffSize=0.07
    OptionalParentItem=(Type=Bone)
    RotationOffset=(X=26,Y=-64.5,Z=0)
    ExecutePin=@n4.B
   )
   n7 = &RigVMFunction_MathFloatMul(B=@n6.OutputParam, A=@n5.Value) {template="Multiply::Execute(in A,in B,out Result)"}
   > &RigUnit_ModifyTransforms(
    ItemToModify=((Item=(Type=Bone,Name="dyn_hair_mstr_ct"),Transform=(Rotation=(X=0,Y=0,Z=0,W=1),Translation=(X=0,Y=0,Z=0),Scale3D=(X=1,Y=1,Z=1))))
    ItemToModify.0.Transform.Rotation=&RigVMFunction_MathQuaternionFromEuler(Euler=(X=0,Y=5,Z=-10)).Result
    Weight=@n7.Result
   )
   > &RigUnit_ModifyTransforms(
    ItemToModify=((Item=(Type=Bone,Name="dyn_hair_04_mstr_r"),Transform=(Rotation=(X=0,Y=0,Z=0,W=1),Translation=(X=0,Y=0,Z=0),Scale3D=(X=1,Y=1,Z=1))))
    ItemToModify.0.Transform.Rotation=&RigVMFunction_MathQuaternionFromEuler(Euler=(X=0,Y=15,Z=-25)).Result
    Weight=@n7.Result
   )
   > &RigUnit_ModifyTransforms(
    ItemToModify=((Item=(Type=Bone,Name="dyn_hair_03_mstr_r"),Transform=(Rotation=(X=0,Y=0,Z=0,W=1),Translation=(X=0,Y=0,Z=0),Scale3D=(X=1,Y=1,Z=1))))
    ItemToModify.0.Transform.Rotation=&RigVMFunction_MathQuaternionFromEuler(Euler=(X=0,Y=-30,Z=0)).Result
    Weight=@n7.Result
   )
  ]}
]
hierarchy [
 root {loc=<0 0 0>, rot=<0 0 0 1>}
 pelvis {parent=root, loc=<0 -0.5301 98.0571>, rot=<0.0127 0.707 -0.0127 -0.707>}
 spine_01 {parent=pelvis, loc=<6.1208 -0.9328 0>, rot=<0 0 0.0576 -0.9983>}
 spine_02 {parent=spine_01, loc=<6.6022 -0.0153 0>, rot=<0 0 0.0022 -1>}
 spine_03 {parent=spine_02, loc=<6.4919 -0.0105 0>, rot=<0 0 0.0015 -1>}
 spine_04 {parent=spine_03, loc=<6.9596 -0.007 0>, rot=<0 0 0.0009 -1>}
 spine_05 {parent=spine_04, loc=<6.9981 -0.0017 0>, rot=<0 0 -0.0042 -1>}
 clavicle_l {parent=spine_05, loc=<5.4831 -0.1053 -3.2946>, rot=<0.0938 -0.7388 0.0307 -0.6666>}
 upperarm_l {parent=clavicle_l, loc=<9.9486 0 0>, rot=<0.0371 -0.2992 -0.0705 -0.9509>}
 upperarm_twist_01_l {parent=upperarm_l, loc=<0.0041 0 0>, rot=<0 0 0 1>}
 neck_01 {parent=spine_05, loc=<11.3138 0.1529 0>, rot=<0 0 -0.2003 -0.9797>}
 neck_02 {parent=neck_01, loc=<4.5627 0 0>, rot=<0 0 0.0504 -0.9987>}
 head {parent=neck_02, loc=<5.3914 0 0>, rot=<0 0 0.115 -0.9934>}
 dyn_hair_01_mstr_l {parent=head, loc=<6.8533 6.2605 -6.1969>, rot=<0.1686 -0.9546 -0.2197 -0.1095>}
 dyn_hair_01_01_l {parent=dyn_hair_01_mstr_l, loc=<0 0 0>, rot=<0 0 0 1>}
 dyn_hair_01_02_l {parent=dyn_hair_01_01_l, loc=<5.9999 0.0102 0>, rot=<-0.005 -0.158 -0.026 -0.9871>}
 dyn_hair_01_03_l {parent=dyn_hair_01_02_l, loc=<3.5809 0 0>, rot=<0 0 -0.2439 -0.9698>}
 dyn_hair_02_mstr_l {parent=head, loc=<7.4887 1.1206 -7.1682>, rot=<0.0866 -0.7861 -0.6116 0.0215>}
 dyn_hair_02_01_l {parent=dyn_hair_02_mstr_l, loc=<0 0 0>, rot=<0 0 0 1>}
 dyn_hair_02_02_l {parent=dyn_hair_02_01_l, loc=<7.0127 0.0046 0>, rot=<0 0 0.0027 -1>}
 dyn_hair_03_mstr_l {parent=head, loc=<7.4888 -2.5674 -5.9957>, rot=<0.0494 -0.8338 -0.5497 0.0139>}
 dyn_hair_03_01_l {parent=dyn_hair_03_mstr_l, loc=<0 0 0>, rot=<0 0 0 1>}
 dyn_hair_03_02_l {parent=dyn_hair_03_01_l, loc=<7.4029 0.0025 0>, rot=<0 0 0.0262 -0.9997>}
 dyn_hair_03_03_l {parent=dyn_hair_03_02_l, loc=<5.7712 0.0477 0>, rot=<0 0 0.2532 -0.9674>}
 dyn_hair_04_mstr_l {parent=head, loc=<7.587 -5.3329 -3.9359>, rot=<0.0753 -0.9972 0 0>}
 dyn_hair_04_01_l {parent=dyn_hair_04_mstr_l, loc=<0 0 0>, rot=<0 0 0 1>}
 dyn_hair_04_02_l {parent=dyn_hair_04_01_l, loc=<7.8602 0 0>, rot=<0 0 0 1>}
 dyn_hair_04_03_l {parent=dyn_hair_04_02_l, loc=<6.2261 0 0>, rot=<0 0 0.1108 -0.9938>}
 dyn_hair_mstr_ct {parent=head, loc=<7.6658 -5.72 0>, rot=<0.0766 -0.9971 0 0>}
 dyn_hair_01_ct {parent=dyn_hair_mstr_ct, loc=<0 0 0>, rot=<0 0 0 1>}
 dyn_hair_02_ct {parent=dyn_hair_01_ct, loc=<8.3841 0 0>, rot=<0 0 0.0027 -1>}
 dyn_hair_03_ct {parent=dyn_hair_02_ct, loc=<6.7499 0 0>, rot=<0 0 0.0523 -0.9986>}
 dyn_hair_01_mstr_r {parent=head, loc=<7.458 4.8417 7.1682>, rot=<0.0863 -0.7861 0.6116 -0.0217>}
 dyn_hair_01_01_r {parent=dyn_hair_01_mstr_r, loc=<0 0 0>, rot=<0 0 0 1>}
 dyn_hair_01_02_r {parent=dyn_hair_01_01_r, loc=<4.9329 0.0128 0>, rot=<0 0 0.0031 -1>}
 dyn_hair_02_mstr_r {parent=head, loc=<7.4888 1.1206 7.1682>, rot=<0.0863 -0.7861 0.6116 -0.0217>}
 dyn_hair_02_01_r {parent=dyn_hair_02_mstr_r, loc=<0 0 0>, rot=<0 0 0 1>}
 dyn_hair_02_02_r {parent=dyn_hair_02_01_r, loc=<7.0062 -1.1019 0>, rot=<0 0 0.0031 -1>}
 dyn_hair_03_mstr_r {parent=head, loc=<7.4893 -2.5674 5.9957>, rot=<0.0493 -0.8338 0.5497 -0.014>}
 dyn_hair_03_01_r {parent=dyn_hair_03_mstr_r, loc=<0 0 0>, rot=<0 0 0 1>}
 dyn_hair_03_02_r {parent=dyn_hair_03_01_r, loc=<7.4032 0 0>, rot=<0 0 0.0223 -0.9998>}
 dyn_hair_03_03_r {parent=dyn_hair_03_02_r, loc=<5.7719 0 0>, rot=<0 0 0.2572 -0.9664>}
 dyn_hair_04_mstr_r {parent=head, loc=<7.5871 -5.3329 3.9359>, rot=<0.0753 -0.9972 0 0>}
 dyn_hair_04_01_r {parent=dyn_hair_04_mstr_r, loc=<0 0 0>, rot=<0 0 0 1>}
 dyn_hair_04_02_r {parent=dyn_hair_04_01_r, loc=<7.8607 -0.0001 -0.0001>, rot=<0 0 0 -1>}
 dyn_hair_04_03_r {parent=dyn_hair_04_02_r, loc=<6.2261 0 0>, rot=<0 0 0.1108 -0.9938>}
 dyn_bangs_01 {parent=head, loc=<8.2773 8.9161 0.1888>, rot=<-0.0789 0.0568 0.9865 0.1321>}
 dyn_bangs_02 {parent=dyn_bangs_01, loc=<2.8608 0 0>, rot=<0.0178 0.0352 -0.0954 -0.9947>}
 dyn_hair_side_01_l {parent=head, loc=<2.1661 2.4856 -6.871>, rot=<0.0866 -0.7861 -0.6116 0.0215>}
 clavicle_r {parent=spine_05, loc=<5.4831 -0.1053 3.2946>, rot=<0.7388 0.0938 -0.6666 -0.0307>}
 upperarm_r {parent=clavicle_r, loc=<-9.9486 0 0>, rot=<0.0371 -0.2992 -0.0705 -0.9509>}
 upperarm_twist_01_r {parent=upperarm_r, loc=<-0.0041 0 0>, rot=<0 0 0 -1>}
]
variables [
 head_tilt_right {type=double, default=0}
 is_sky_diving {type=bool, default=False}
]
