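# Physics asset for the F_MED_Barium character's hair and face accessory mesh, defining capsule bodies and joint limits for its dynamic hair bones.
# Left, right and center hair chains run from small master bodies, while the bangs and a left side lock attach to the head, all with narrow swing and very tight twist limits.
# Head, neck, upper spine, clavicle and upper-arm capsules add body shapes; only the right clavicle and arm are constrained, with free angular motion.
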PhysicsAsset /BRCosmetics/Characters/Player/Female/Medium/Bodies/F_MED_Barium/Meshes/Parts/F_MED_Barium_Hair_FaceAcc_Physics
bodies {
 clavicle_l [
  {
   shapeType=Capsule
   length=9.8738
   radius=5.4417
   shapeName=clavicle_l_capsule
   center=<5.1414 0.0001 -0.273>
   rotation=<rot 89.3528 -0.0509 -0.0519>
  }
  {
   shapeType=Capsule
   length=8.8796
   radius=3.5998
   shapeName=clavicle_l_capsule
   center=<1.867 -2.4742 1.796>
   rotation=<rot 70 -0.0001 -0.0001>
  }
 ]
 clavicle_r [
  {
   shapeType=Capsule
   length=9.8738
   radius=5.4417
   shapeName=clavicle_l_capsule
   center=<-5.1414 -0.0001 0.273>
   rotation=<rot 89.3528 -0.0509 179.9481>
  }
  {
   shapeType=Capsule
   length=8.8796
   radius=3.5998
   shapeName=clavicle_l_capsule
   center=<-1.867 2.4742 -1.796>
   rotation=<rot 70 -0.0001 179.9999>
  }
 ]
 dyn_bangs_01 [{shapeType=Capsule, length=2.3114, radius=0.7661, shapeName=dyn_bangs_01_capsule, center=<1.5973 0 0>, rotation=<rot 90 0 0.0005>}]
 dyn_bangs_02 [{shapeType=Capsule, length=3.3781, radius=0.5376, shapeName=dyn_bangs_02_capsule, center=<1.7387 0 0>, rotation=<rot 90 179.9999 -179.9994>}]
 dyn_hair_01_01_l [{
   shapeType=Capsule
   length=2.3089
   radius=1.5124
   shapeName=dyn_hair_01_01_l_capsule
   center=<2.011 0 0.4297>
   rotation=<rot -90 4.0447 -4.0445>
  }]
 dyn_hair_01_01_r [{shapeType=Capsule, length=4.9387, radius=1.6734, shapeName=dyn_hair_01_01_r_capsule, center=<3.7847 -1.0254 0>, rotation=<rot 90 0 -0.0001>}]
 dyn_hair_01_02_l [
  {
   shapeType=Capsule
   length=4.5173
   radius=1.1685
   shapeName=dyn_hair_01_02_l_capsule
   center=<1.3562 0 0.4297>
   rotation=<rot -90 -179.9801 179.9804>
  }
  {
   shapeType=Capsule
   length=4.4276
   radius=0.9401
   shapeName=dyn_hair_01_02_l_capsule
   center=<1.2339 1.8711 -0.9588>
   rotation=<rot -74.9999 0.0006 2.5298>
  }
 ]
 dyn_hair_01_02_r [{
   shapeType=Capsule
   length=5.2674
   radius=1.3447
   shapeName=dyn_hair_01_02_r_capsule
   center=<4.3322 -1.0254 0>
   rotation=<rot 90 179.9605 179.9604>
  }]
 dyn_hair_01_03_l [{shapeType=Capsule, length=2.8088, radius=1.165, shapeName=dyn_hair_01_03_l_capsule, center=<1.8694 0 0.4297>, rotation=<rot -90 0 0.0004>}]
 dyn_hair_01_ct [{shapeType=Capsule, length=5.4978, radius=2, shapeName=dyn_hair_01_ct_capsule, center=<4.1411 -0.3313 0>, rotation=<rot 90 169.3803 174.3789>}]
 dyn_hair_01_mstr_l [{shapeType=Capsule, length=0, radius=0.5, shapeName=dyn_hair_01_mstr_l_capsule}]
 dyn_hair_01_mstr_r [{shapeType=Capsule, length=0, radius=0.5, shapeName=dyn_hair_01_mstr_r_capsule}]
 dyn_hair_02_01_l [{
   shapeType=Capsule
   length=5.3389
   radius=1.4294
   shapeName=dyn_hair_02_01_l_capsule
   center=<5.3286 -0.9977 0>
   rotation=<rot 90 180 179.9998>
  }]
 dyn_hair_02_01_r [{
   shapeType=Capsule
   length=5.9379
   radius=1.9892
   shapeName=dyn_hair_02_01_r_capsule
   center=<4.8517 -1.6704 0>
   rotation=<rot -90 -180 -179.9999>
  }]
 dyn_hair_02_02_l [{
   shapeType=Capsule
   length=4.7683
   radius=2
   shapeName=dyn_hair_02_02_l_capsule
   center=<3.7924 -0.4968 0.5048>
   rotation=<rot 90 -90 -90.0002>
  }]
 dyn_hair_02_02_r [{
   shapeType=Capsule
   length=5.9379
   radius=1.917
   shapeName=dyn_hair_02_02_r_capsule
   center=<4.8601 -0.2301 0>
   rotation=<rot -90 180 177.6015>
  }]
 dyn_hair_02_ct [{shapeType=Capsule, length=6.3035, radius=2, shapeName=dyn_hair_02_ct_capsule, center=<3.7924 -0.7325 0>, rotation=<rot 90 179.9999 179.9994>}]
 dyn_hair_02_mstr_l [{shapeType=Capsule, length=0, radius=0.5, shapeName=dyn_hair_02_mstr_l_capsule}]
 dyn_hair_02_mstr_r [{shapeType=Capsule, length=0, radius=0.5, shapeName=dyn_hair_02_mstr_r_capsule}]
 dyn_hair_03_01_l [{shapeType=Capsule, length=4.7683, radius=2, shapeName=dyn_hair_03_01_l_capsule, center=<3.7924 0 0>, rotation=<rot 90 0 -0.0002>}]
 dyn_hair_03_01_r [{shapeType=Capsule, length=4.7683, radius=2, shapeName=dyn_hair_03_01_l_capsule, center=<3.7924 0 0>, rotation=<rot 90 0 179.9998>}]
 dyn_hair_03_02_l [{
   shapeType=Capsule
   length=4.7683
   radius=2
   shapeName=dyn_hair_03_02_l_capsule
   center=<3.7924 -0.0903 -0.0449>
   rotation=<rot 90 180 179.9997>
  }]
 dyn_hair_03_02_r [{shapeType=Capsule, length=2.8023, radius=2, shapeName=dyn_hair_03_02_l_capsule, center=<2.9012 0 0>, rotation=<rot 90 180 -0.0003>}]
 dyn_hair_03_03_l [
  {
   shapeType=Capsule
   length=3.0053
   radius=0.6298
   shapeName=dyn_hair_03_03_l_capsule
   center=<3.7924 -0.096 0.4788>
   rotation=<rot 90 -10.0252 -10.0257>
  }
  {
   shapeType=Capsule
   length=3.0053
   radius=0.6298
   shapeName=dyn_hair_03_03_l_capsule
   center=<3.7924 -0.096 -0.6534>
   rotation=<rot 90 -90 -90.0005>
  }
 ]
 dyn_hair_03_03_r [
  {
   shapeType=Capsule
   length=4.65
   radius=1.3326
   shapeName=dyn_hair_03_03_r_capsule
   center=<2.7184 0.1897 1.1337>
   rotation=<rot -90 -0.0003 0.0008>
  }
  {
   shapeType=Capsule
   length=4.65
   radius=1.3326
   shapeName=dyn_hair_03_03_r_capsule
   center=<2.7184 0.1897 -1.1867>
   rotation=<rot -90 0 0.0005>
  }
 ]
 dyn_hair_03_ct [
  {
   shapeType=Capsule
   length=6.7587
   radius=0.8229
   shapeName=dyn_hair_04_03_l_capsule
   center=<2.7662 0.2184 0.9986>
   rotation=<rot 90 179.9998 179.9993>
  }
  {
   shapeType=Capsule
   length=6.7587
   radius=0.8229
   shapeName=dyn_hair_04_03_l_capsule
   center=<2.7662 0.2184 -0.5056>
   rotation=<rot 90 179.9999 179.9994>
  }
 ]
 dyn_hair_03_mstr_l [{shapeType=Capsule, length=0, radius=0.5, shapeName=dyn_hair_03_mstr_l_capsule}]
 dyn_hair_03_mstr_r [{shapeType=Capsule, length=0, radius=0.5, shapeName=dyn_hair_03_mstr_r_capsule}]
 dyn_hair_04_01_l [{shapeType=Capsule, length=4.7683, radius=2, shapeName=dyn_hair_04_01_l_capsule, center=<3.7924 0 0>, rotation=<rot 90 179.9803 179.9798>}]
 dyn_hair_04_01_r [{shapeType=Capsule, length=4.7683, radius=2, shapeName=dyn_hair_04_01_l_capsule, center=<3.7924 0 0>, rotation=<rot 90 -89.9127 95.0859>}]
 dyn_hair_04_02_l [{shapeType=Capsule, length=4.7683, radius=2, shapeName=dyn_hair_04_02_l_capsule, center=<3.7924 0 0>, rotation=<rot 90 179.9803 179.9798>}]
 dyn_hair_04_02_r [{shapeType=Capsule, length=4.7683, radius=2, shapeName=dyn_hair_04_02_l_capsule, center=<3.7924 0 0>, rotation=<rot 90 179.9977 -0.0028>}]
 dyn_hair_04_03_l [
  {
   shapeType=Capsule
   length=6.7587
   radius=0.8229
   shapeName=dyn_hair_04_03_l_capsule
   center=<2.7662 0.2184 0.6602>
   rotation=<rot 90 180 179.9995>
  }
  {
   shapeType=Capsule
   length=6.7587
   radius=0.8229
   shapeName=dyn_hair_04_03_l_capsule
   center=<2.7662 0.2184 -0.8041>
   rotation=<rot 90 179.9998 179.9993>
  }
 ]
 dyn_hair_04_03_r [
  {
   shapeType=Capsule
   length=6.7587
   radius=0.8229
   shapeName=dyn_hair_04_03_l_capsule
   center=<2.7662 0.2184 0.6602>
   rotation=<rot 90 180 179.9995>
  }
  {
   shapeType=Capsule
   length=6.7587
   radius=0.8229
   shapeName=dyn_hair_04_03_l_capsule
   center=<2.7662 0.2184 -0.9063>
   rotation=<rot 90 179.9998 179.9993>
  }
 ]
 dyn_hair_04_mstr_l [{shapeType=Capsule, length=0, radius=0.5, shapeName=dyn_hair_04_mstr_l_capsule}]
 dyn_hair_04_mstr_r [{shapeType=Capsule, length=0, radius=0.5, shapeName=dyn_hair_04_mstr_r_capsule}]
 dyn_hair_mstr_ct [{shapeType=Capsule, length=0, radius=0.5, shapeName=dyn_hair_mstr_ct_capsule}]
 dyn_hair_side_01_l [{
   shapeType=Capsule
   length=2.653
   radius=1.1023
   shapeName=dyn_hair_02_01_l_capsule
   center=<1.866 -0.1807 -0.3916>
   rotation=<rot 0.6409 92.4602 -104.5928>
  }]
 head [
  {shapeType=Capsule, length=2.8871, radius=6.3917, shapeName=head_capsule, center=<4.5578 1.2063 0>, rotation=<rot -90 -165.9638 75.9624>}
  {shapeType=Capsule, length=6.6476, radius=5.372, shapeName=head_capsule, center=<1.4968 4.2258 0>, rotation=<rot 90 180 -161.7126>}
  {shapeType=Capsule, length=7.8665, radius=3.1448, shapeName=head_capsule, center=<1.8101 0.1627 0.0229>, rotation=<rot -0.7246 179.9993 -180>}
 ]
 neck_01 [
  {shapeType=Capsule, length=8.0279, radius=5.109, shapeName=neck_01_capsule, center=<0 0.1709 0>, rotation=<rot 90 -179.9928 -179.9924>}
  {
   shapeType=Capsule
   length=22.9773
   radius=14.9278
   shapeName=neck_01_capsule
   center=<7.1978 16.7247 0.0084>
   rotation=<rot 0 -0.0006 -2.3987>
  }
 ]
 spine_05 [{shapeType=Capsule, length=8.0222, radius=8.6509, shapeName=spine_05_capsule, center=<2.9208 0.3835 0>, rotation=<rot 0 179.9994 -180>}]
 upperarm_l [{shapeType=Capsule, length=24.4397, radius=5.5603, shapeName=upperarm_l_capsule, center=<13.2466 0 0>, rotation=<rot -90 179.9604 -179.9605>}]
 upperarm_r [{shapeType=Capsule, length=24.4397, radius=5.5603, shapeName=upperarm_l_capsule, center=<-13.2466 0 0>, rotation=<rot -90 90 89.9999>}]
}
constraints [
 {bone1Name=dyn_hair_02_01_l, bone2Name=dyn_hair_02_mstr_l, swing1LimitDegrees=15, swing2LimitDegrees=20, twistLimitDegrees=2}
 {bone1Name=dyn_hair_02_02_l, bone2Name=dyn_hair_02_01_l, swing1LimitDegrees=10, swing2LimitDegrees=10, twistLimitDegrees=2}
 {bone1Name=dyn_hair_03_01_l, bone2Name=dyn_hair_03_mstr_l, swing1LimitDegrees=15, swing2LimitDegrees=10, twistLimitDegrees=2}
 {bone1Name=dyn_hair_03_02_l, bone2Name=dyn_hair_03_01_l, swing1LimitDegrees=10, swing2LimitDegrees=10, twistLimitDegrees=2}
 {bone1Name=dyn_hair_03_03_l, bone2Name=dyn_hair_03_02_l, swing1LimitDegrees=20, swing2LimitDegrees=10, twistLimitDegrees=2}
 {bone1Name=dyn_hair_04_01_l, bone2Name=dyn_hair_04_mstr_l, swing1LimitDegrees=20, swing2LimitDegrees=10, twistLimitDegrees=2}
 {bone1Name=dyn_hair_04_02_l, bone2Name=dyn_hair_04_01_l, swing1LimitDegrees=10, swing2LimitDegrees=10, twistLimitDegrees=2}
 {bone1Name=dyn_hair_04_03_l, bone2Name=dyn_hair_04_02_l, swing1LimitDegrees=20, swing2LimitDegrees=10, twistLimitDegrees=2}
 {bone1Name=dyn_hair_01_ct, bone2Name=dyn_hair_mstr_ct, swing1LimitDegrees=20, swing2LimitDegrees=10, twistLimitDegrees=2}
 {bone1Name=dyn_hair_02_ct, bone2Name=dyn_hair_01_ct, swing1LimitDegrees=10, swing2LimitDegrees=10, twistLimitDegrees=2}
 {bone1Name=dyn_hair_03_ct, bone2Name=dyn_hair_02_ct, swing1LimitDegrees=20, swing2LimitDegrees=10, twistLimitDegrees=2}
 {bone1Name=dyn_hair_03_01_r, bone2Name=dyn_hair_03_mstr_r, swing1LimitDegrees=15, swing2LimitDegrees=10, twistLimitDegrees=2}
 {bone1Name=dyn_hair_03_02_r, bone2Name=dyn_hair_03_01_r, swing1LimitDegrees=15, swing2LimitDegrees=20, twistLimitDegrees=2}
 {bone1Name=dyn_hair_03_03_r, bone2Name=dyn_hair_03_02_r, swing1LimitDegrees=20, swing2LimitDegrees=10, twistLimitDegrees=2}
 {bone1Name=dyn_hair_02_01_r, bone2Name=dyn_hair_02_mstr_r, swing1LimitDegrees=15, swing2LimitDegrees=15, twistLimitDegrees=2}
 {bone1Name=dyn_hair_02_02_r, bone2Name=dyn_hair_02_01_r, swing1LimitDegrees=10, swing2LimitDegrees=10, twistLimitDegrees=2}
 {bone1Name=dyn_hair_01_01_r, bone2Name=dyn_hair_01_mstr_r, swing1LimitDegrees=15, swing2LimitDegrees=15, twistLimitDegrees=2}
 {bone1Name=dyn_hair_01_02_r, bone2Name=dyn_hair_01_01_r, swing1LimitDegrees=10, swing2LimitDegrees=10, twistLimitDegrees=2}
 {bone1Name=dyn_hair_01_01_l, bone2Name=dyn_hair_01_mstr_l, swing1LimitDegrees=20, swing2LimitDegrees=15, twistLimitDegrees=2}
 {bone1Name=dyn_hair_01_02_l, bone2Name=dyn_hair_01_01_l, swing1LimitDegrees=15, swing2LimitDegrees=10, twistLimitDegrees=2}
 {bone1Name=dyn_hair_01_03_l, bone2Name=dyn_hair_01_02_l, swing1LimitDegrees=20, swing2LimitDegrees=10, twistLimitDegrees=2}
 {bone1Name=dyn_hair_04_02_r, bone2Name=dyn_hair_04_01_r, swing1LimitDegrees=20, swing2LimitDegrees=10, twistLimitDegrees=2}
 {bone1Name=dyn_hair_04_01_r, bone2Name=dyn_hair_04_mstr_r, swing1LimitDegrees=20, swing2LimitDegrees=10, twistLimitDegrees=2}
 {bone1Name=dyn_hair_04_03_r, bone2Name=dyn_hair_04_02_r, swing1LimitDegrees=20, swing2LimitDegrees=10, twistLimitDegrees=2}
 {bone1Name=dyn_bangs_01, bone2Name=head, swing1LimitDegrees=10, swing2LimitDegrees=15, twistLimitDegrees=2}
 {bone1Name=dyn_bangs_02, bone2Name=dyn_bangs_01, swing1LimitDegrees=10, swing2LimitDegrees=10, twistLimitDegrees=2}
 {
  bone1Name=clavicle_r
  bone2Name=spine_05
  swing1LimitDegrees=45
  swing1Motion=Free
  swing2LimitDegrees=45
  swing2Motion=Free
  twistLimitDegrees=45
  twistMotion=Free
 }
 {
  bone1Name=upperarm_r
  bone2Name=clavicle_r
  swing1LimitDegrees=45
  swing1Motion=Free
  swing2LimitDegrees=45
  swing2Motion=Free
  twistLimitDegrees=45
  twistMotion=Free
 }
 {bone1Name=dyn_hair_side_01_l, bone2Name=head, swing1LimitDegrees=20, swing2LimitDegrees=20, twistLimitDegrees=2}
]
