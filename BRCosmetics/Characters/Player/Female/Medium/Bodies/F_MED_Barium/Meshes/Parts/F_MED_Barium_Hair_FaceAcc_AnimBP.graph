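# Animation Blueprint for the F_MED_Barium hair and face accessory part, following the attached parent mesh and adding secondary motion.
# The copied parent pose feeds a Control Rig driven by skydiving state and head tilt, then a RigidBody simulation in head-bone space.
# Each update computes the simulation's gravity override with a deprecated state function and derives right head tilt from head twist relative to spine_05.
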AnimBlueprint /BRCosmetics/Characters/Player/Female/Medium/Bodies/F_MED_Barium/Meshes/Parts/F_MED_Barium_Hair_FaceAcc_AnimBP
target_skeleton=/Game/Characters/Player/Female/Medium/Bodies/F_MED_Barium/Meshes/Parts/F_MED_Barium_Hair_FaceAcc_Skeleton
graphs [
 AnimGraph {nodes [
   n1 = Control Rig!(
    Source=Copy Pose From Mesh() {bUseAttachedParent=true}
    Alpha
    is_sky_diving:bool=Get bIsSkyDiving()
    head_tilt_right:double=Get head_tilt_right() -> (head_tilt_right:double)
   ) -> (Pose) {
    alphaBoolBlend {blendOption=Linear}
    controlRigAssetReference {blueprintRigClass=&/F_MED_Barium_Hair_FaceAcc_CtrlRig.F_MED_Barium_Hair_FaceAcc_CtrlRig_C}
   }
   n2 = RigidBody!(
    ComponentPose=Local To Component(LocalPose=@n1)
    Alpha
    OverrideWorldGravity:Vector=Get gravity_override() -> (gravity_override:Vector)
    ExternalForce
   ) -> (Pose) {
    bClampLinearTranslationLimitToRefPose=true
    bOverrideWorldGravity=true
    overridePhysicsAsset=&/F_MED_Barium_Hair_FaceAcc_Physics
    simulationSpace=BaseBoneSpace
    alphaBoolBlend {blendOption=Linear}
    baseBoneRef {boneName=head}
    componentAppliedLinearAccClamp=<3000 3000 3000>
    componentLinearAccScale=<0.4 0.3 0.3>
    componentLinearVelScale=<0.4 0.3 0.3>
    simSpaceSettings {maxAngularAcceleration=1000, maxAngularVelocity=3, maxLinearAcceleration=250, maxLinearVelocity=500, velocityScaleZ=0, worldAlpha=1}
   }
   Output Pose(Result=Component To Local(ComponentPose=@n2)) {layerGroup=None}
  ]}
 EventGraph {nodes [
   Event BlueprintUpdateAnimation()
   > n1 = Sequence()
   > n2 = Gravity Override State Function DEPRECATED(CharacterPartAnimInstance=Self-Reference(), GameTimeInSeconds=GetGameTimeInSeconds(), GravityOverrideParams=Get Gravity Override Params())
   > Set gravity_override(gravity_override=@n2)
   Set head_tilt_right(
    execute=@n1.then_1
    head_tilt_right:double=GetAlphaFromReferenceTransformDelta(
     SkeletalMeshComponent=Get skeletal_mesh() -> (skeletal_mesh:Skeletal Mesh Component Object Reference)
     BoneName=head
     BaseName=spine_05
     DeltaComponentType=QuaternionTwist
     InputRange=(X=10,Y=30)
     TwistAxis="0,1,0"
    )
   ) -> (Output_Get:double)
   Event BlueprintInitializeAnimation()
   > Set skeletal_mesh(skeletal_mesh:Skeletal Mesh Component Object Reference=GetOwnerSkeletalMeshComponent()) -> (Output_Get:Skeletal Mesh Component Object Reference)
  ]}
]
variables [
 Gravity Override Params {type=GravityOverrideParamsStruct, default {onGround_Moving {gravityOverride=<0 -75 -980>}}}
 gravity_override {type=Vector}
 head_tilt_right {type=number}
 skeletal_mesh {type=SkeletalMeshComponent}
]
